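# Parent material for Academia superlevel character hair, built between the shared 2023 character PreFX and PostFX functions.
# Fortnite Hair 2.0 output is mask-blended over the base, with an optional fly-away hair pass, both behind static switches.
# Superlevel parameters desaturate, hue-shift, scale, offset and tint base color and emissive alike.
# A colored Fresnel rim is added to emissive, and metallic is replaced by an override before PostFX.
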Material /BRCosmetics/Characters/Player/Female/Medium/Bodies/F_MED_SpatialTravel_Watt/Skins/Superlevels/ParentMaterials/M_AcademiaSL_HairParent
graph {
 properties {
  bUseMaterialAttributes=true
  bUsedWithInstancedStaticMeshes=true
  bUsedWithMeshDeformer=true
  bUsedWithMorphTargets=true
  bUsedWithSkeletalMesh=true
  pixelDepthOffsetMode=PDOM_Legacy
  shadingModel=MSM_FromMaterialExpression
  usedShadingModels=Default Lit | Subsurface | Clear Coat | Subsurface Profile | Cloth
 }
 nodes [
  n1 = MaterialFunctionCall() {materialFunction=&MF_CharacterParent_2023_PreFX}
  n2 = MaterialFunctionCall!(In=@n1.Result, Mask Input=@n1.Result) {materialFunction=/Game/Global/Materials/Character_Parent_2023/Hair_Parent_2023/MaterialFunctions/MF_FN_Hair_2023}
  n3 = BlendMaterialAttributes(
   A=@n1.Result
   B=StaticSwitchParameter(True=@n2.Result, False=@n1.Result) {defaultValue=true, group=Skin FX, parameterName=UseFNHair2_0}
   Alpha=@n2.HairMask
  )
  n4 = GetMaterialAttributes(None=StaticSwitchParameter(
    True=MaterialFunctionCall!(MaterialInput=@n3) {materialFunction=/Game/Global/Materials/Character_Parent_2023/Hair_Parent_2023/MaterialFunctions/MF_FlyAwayHair_2023}.Result
    False=@n3
   ) {group=Skin FX, parameterName=IsFlyAwayHair?}) {attributeGetTypes [#1, #2]}
  n5 = ScalarParameter() {group=Superlevel, parameterName=HueShift}
  n6 = ScalarParameter() {group=Superlevel, parameterName=Desaturation}
  n7 = Desaturation(None=@n4.Base Color, Fraction=@n6)
  n8 = ScalarParameter() {defaultValue=1, group=Superlevel, parameterName=HairColor_Multiplier}
  n9 = ScalarParameter() {group=Superlevel, parameterName=HairColor_AdditiveOffset}
  n10 = Add(
   A=Multiply(
    A=MaterialFunctionCall(Hue Shift Percentage=@n5, Texture=@n7) {materialFunction=&HueShift}.Result
    B=@n8
   )
   B=@n9
  )
  n11 = VectorParameter() {group=Superlevel, parameterName=HairColor_Tint, defaultValue=<rgba 1 1 1 0>}
  n12 = MaterialFunctionCall(Hue Shift Percentage=@n5, Texture=Desaturation(None=@n4.Emissive Color, Fraction=@n6)) {materialFunction=&HueShift}
  n13 = Multiply(A=Add(A=Multiply(A=@n12.Result, B=@n8), B=@n9), B=@n11.RGB)
  n14 = Multiply(
   A=Power(Base=NamedRerouteUsage(), Exp=ScalarParameter() {defaultValue=6, parameterName=Hair_AdditionalFresnelPower})
   B=VectorParameter() {parameterName=Hair_AdditionalFresnelColor, defaultValue=<rgba 0.5469 0.1001 0.3395 1>}.RGB
  )
  MaterialFunctionCall(
   Material=SetMaterialAttributes!(
    MaterialAttributes=@n4.MaterialAttributes
    Base Color=Multiply(A=@n10, B=@n11.RGB)
    Metallic=ScalarParameter() {defaultValue=1, parameterName=MetallicOverride}
    Emissive Color=Add(A=@n13, B=@n14)
   ) {attributeSetTypes [#1, #3, #2]}
   M=@n1.M
   Pre Skinned Local Position=@n1.M
   Distance Blend=@n1.M
   Fresnel=NamedRerouteDeclaration(None=@n1.Fresnel) {name=InFresnel, nodeColor=<rgba 1 0 0.1875 1>}
  ) {materialFunction=&MF_CharacterParent_2023_PostFX}
  Reroute(None=@n1.Tech Art Mask)
 ]
}
