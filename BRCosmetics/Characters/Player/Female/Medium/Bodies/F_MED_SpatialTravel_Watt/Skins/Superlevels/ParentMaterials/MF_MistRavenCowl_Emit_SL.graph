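# Material function for the MistRaven cowl character skin that tints base color and adds emissive glow to incoming material attributes, behind a static switch.
# Body, neck and face colors multiply base color, while emissive combines masked tints, a hand color blend and masked head darkening.
# Glow brightness follows time of day and shading path, with Fresnel power blended from leg to neck by pre-skinned height.
# Extra outputs expose the emissive, cowl and body masks.
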MaterialFunction /BRCosmetics/Characters/Player/Female/Medium/Bodies/F_MED_SpatialTravel_Watt/Skins/Superlevels/ParentMaterials/MF_MistRavenCowl_Emit_SL
graph {nodes [
  n1 = NamedRerouteUsage()
  n2 = NamedRerouteUsage()
  n3 = GetMaterialAttributes(None=FunctionInput() {inputName=Material, inputType=FunctionInput_MaterialAttributes}) {attributeGetTypes [#1]}
  n4 = VectorParameter() {group=MistRaven, parameterName=NeckColor, defaultValue=<rgba 0.2 0.0596 0.1744 1>}
  n5 = LinearInterpolate(
   A=@n4.RGB
   B=VectorParameter() {group=MistRaven, parameterName=FaceColor, defaultValue=<rgba 0.2 0.0596 0.1744 1>}.RGB
   Alpha=SphereMask(A=TextureCoordinate(), B=Constant() {r=0.5}, Radius=Constant() {r=0.5}, Hardness=Constant())
  )
  n6 = TextureSampleParameter2D() {
   group=MistRaven
   parameterName=BodyMask
   samplerSource=SSM_Wrap_WorldGroupSettings
   samplerType=SAMPLERTYPE_Masks
   texture=&T_BlackMask
  }
  n7 = LinearInterpolate(A=VectorParameter() {group=MistRaven, parameterName=BodyColor, defaultValue=<rgba 0.0187 0.0046 0.026 1>}.RGB, B=@n4.RGB, Alpha=@n6.B)
  n8 = LinearInterpolate(
   A=@n3.BaseColor
   B=Multiply(A=@n3.BaseColor, B=StaticSwitchParameter(True=@n5, False=@n7) {group=Styx_Transit, parameterName=IsNeck})
   Alpha=NamedRerouteUsage()
  )
  n9 = ComponentMask(None=NamedRerouteUsage()) {b=true}
  n10 = LinearInterpolate(A=@n8, B=LinearInterpolate(A=@n8, B=Multiply(A=@n8) {constB=0.1}, Alpha=@n9), Alpha=ScalarParameter() {parameterName=IsHead}) {constAlpha=0}
  FunctionOutput(None=StaticSwitchParameter(
    True=SetMaterialAttributes!(MaterialAttributes=@n1, EmissiveColor=@n2, BaseColor=@n10) {attributeSetTypes [#2, #1]}
    False=NamedRerouteUsage()
   ) {defaultValue=true, group=04_Tech Art, parameterName=UseBasicEmissive_Cowl_Body, sortPriority=0})
  n11 = TextureSampleParameter2D() {
   group=06_Emissive
   parameterName=Emissive Mask
   samplerSource=SSM_Wrap_WorldGroupSettings
   samplerType=SAMPLERTYPE_Masks
   texture=&T_BlackMask
  }
  n12 = NamedRerouteUsage()
  n13 = ScalarParameter() {defaultValue=0.85, group=06_Emissive, parameterName=Fresnel_Processing_power}
  FunctionOutput(None=Multiply(A=@n11.B, B=Power(Base=@n12, Exp=@n13) {constExponent=10})) {outputName=Emissive Mask}
  n14 = TextureSample() {
   samplerSource=SSM_Wrap_WorldGroupSettings
   samplerType=SAMPLERTYPE_Masks
   texture=/BRCosmetics/Characters/Player/Female/Medium/Bodies/F_MED_MistRaven_Peck/Skins/Cowl/Textures/T_MistRaven_Peck_Body_TA_mask4
  }
  FunctionOutput(None=NamedRerouteDeclaration(None=@n14.RGB) {name=CowlMask, nodeColor=<rgba 0 0.9297 1 1>}) {outputName=CowlMask, sortPriority=11}
  FunctionOutput(None=@n6.RGB) {outputName=BodyMask, sortPriority=100}
  FunctionOutput(None=@n11.B) {outputName=EmissiveBlueChannel}
  n15 = MaterialFunctionCall(
   Value_Morning=ScalarParameter() {defaultValue=1, group=MistRaven, parameterName=EmissiveBrightness_Morning}
   Value_Day=ScalarParameter() {defaultValue=1, group=MistRaven, parameterName=EmissiveBrightness_day}
   Value_Evening=ScalarParameter() {defaultValue=1, group=MistRaven, parameterName=EmissiveBrightness_Evening}
   Value_Night=ScalarParameter() {defaultValue=1, group=MistRaven, parameterName=EmissiveBrightness_Night}
  ) {materialFunction=&Scalar_Time_of_Day}
  n16 = ScalarParameter() {defaultValue=1, group=06_Emissive, parameterName=EmissiveBrightness}
  n17 = ScalarParameter() {defaultValue=1, group=06_Emissive, parameterName=EmissiveMobileMultiplier}
  n18 = Constant() {r=1}
  NamedRerouteDeclaration(None=Multiply(A=Multiply(A=@n15.Result, B=@n16), B=ShadingPathSwitch(Default=@n17, Deferred=@n18))) {name=Emissive Brightness Modulation, nodeColor=<rgba 1 0.3281 0 1>}
  n19 = Constant() {r=1}
  n20 = Fresnel(ExponentIn=Constant() {r=1}) {baseReflectFraction=0}
  n21 = Saturate(None=Divide(A=FunctionInput() {inputName=PreskinnedPos}) {constB=100})
  n22 = Power(
   Base=StaticSwitchParameter(True=@n20, False=OneMinus(None=@n20)) {defaultValue=true, group=06_Emissive, parameterName=InvertEmissiveFresnel}
   Exp=LinearInterpolate(
    A=ScalarParameter() {defaultValue=0.8325, group=MistRaven, parameterName=EmissiveFresnelPower_leg}
    B=ScalarParameter() {defaultValue=1.627, group=06_Emissive, parameterName=EmissiveFresnelPower_neck}
    Alpha=ComponentMask(None=@n21) {b=true}
   ) {constB=21}
  )
  NamedRerouteDeclaration(None=Saturate(None=Multiply(A=@n19, B=@n22))) {name=Fresnel Processing, nodeColor=<rgba 0 0.7187 1 1>}
  NamedRerouteDeclaration(None=@n11.B) {name=Emissive Mask, nodeColor=<rgba 1 0 0.75 1>}
  n23 = NamedRerouteUsage()
  n24 = Multiply(A=@n23, B=@n23)
  n25 = Multiply(
   A=Multiply(A=NamedRerouteUsage(), B=StaticSwitchParameter(True=@n24, False=@n23) {group=06_Emissive, parameterName=UseLinearEmissiveMask})
   B=NamedRerouteUsage()
  )
  n26 = NamedRerouteUsage()
  NamedRerouteDeclaration(None=StaticSwitchParameter(True=Multiply(A=@n25, B=@n26), False=@n25) {defaultValue=true, group=06_Emissive, parameterName=UseEmissiveFresnel_Cowl_Body}) {name=Main_Emissive, nodeColor=<rgba 0 0.0156 1 1>}
  n27 = VectorParameter() {group=MistRaven, parameterName=Handcolor, defaultValue=<rgba 0.6172 0.0723 0.1356 1>}
  n28 = MaterialFunctionCall(Value_Morning=Constant() {r=0.85}, Value_Day=Constant() {r=1}, Value_Evening=Constant() {r=0.85}, Value_Night=Constant() {r=0.7}) {materialFunction=&Scalar_Time_of_Day}
  n29 = Multiply(A=ComponentMask(None=NamedRerouteUsage()) {r=true}, B=NamedRerouteUsage())
  n30 = ScalarParameter() {group=MistRaven, parameterName=HandColorBlend}
  NamedRerouteDeclaration(None=LinearInterpolate(A=NamedRerouteUsage(), B=Multiply(A=@n27.RGB, B=@n28.Result), Alpha=Multiply(A=@n29, B=@n30))) {name=Hand_Emissive, nodeColor=<rgba 0.3359 1 0 1>}
  NamedRerouteDeclaration(None=@n3.MaterialAttributes) {name=Input_Material, nodeColor=<rgba 1 0 0.6562 1>}
  n31 = NamedRerouteUsage()
  NamedRerouteDeclaration(None=LinearInterpolate(A=NamedRerouteUsage(), B=LinearInterpolate(A=@n31, Alpha=@n9) {constB=0}, Alpha=ScalarParameter() {parameterName=IsHead}) {constAlpha=0}) {name=Final Emissive, nodeColor=<rgba 0 0.9766 1 1>}
  n32 = VectorParameter() {group=MistRaven, parameterName=EmissiveBaseTint, sortPriority=6, defaultValue=<rgba 0.22 0.011 0.0113 1>}
  n33 = VectorParameter() {group=MistRaven, parameterName=EmissiveTint_neck, sortPriority=6, defaultValue=<rgba 0.4522 0.0226 0.0413 1>}
  NamedRerouteDeclaration(None=LinearInterpolate(A=@n32.RGB, B=@n33.RGB, Alpha=@n6.B)) {name=BaseEmissive, nodeColor=<rgba 1 0.7031 0 1>}
  NamedRerouteDeclaration(None=Constant() {r=1}) {name=FaceMask, nodeColor=<rgba 0 0.7656 1 1>}
 ]}
