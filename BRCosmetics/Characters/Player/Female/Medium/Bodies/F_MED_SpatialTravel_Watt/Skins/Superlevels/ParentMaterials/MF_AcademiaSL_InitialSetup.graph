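# Material function that sets up the base material attributes for "AcademiaSL" superlevel skins, reusing effects from other characters' skins.
# It runs input attributes through the AshenMagnus chest emissive function, outputs the resulting base color and emissive as originals, then zeroes emissive.
# Static switch UseLazyLizzSkin adds a UV2 composite and LazyLizz skin layer.
# UseMistRavenSkin then chains MistRaven cowl emissive and Trans functions, driven by pre-skinned local position.
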MaterialFunction /BRCosmetics/Characters/Player/Female/Medium/Bodies/F_MED_SpatialTravel_Watt/Skins/Superlevels/ParentMaterials/MF_AcademiaSL_InitialSetup
graph {nodes [
  n1 = Constant()
  n2 = GetMaterialAttributes(None=MaterialFunctionCall!(
    AshenMagnus_OriginalEmissive=FunctionInput(Preview=@n1) {bUsePreviewValueAsDefault=true, inputName=AshenMagnus_OriginalEmissive, inputType=FunctionInput_Scalar}
    Material=FunctionInput() {inputName=Result, inputType=FunctionInput_MaterialAttributes}
    Mask
   ) {materialFunction=/BRCosmetics/Characters/Player/Male/Medium/Bodies/M_MED_AshenMagnus_Slant/Materials/MF_AshenMagnus_ChestEmissive_SL}.Result) {attributeGetTypes [#1, #2]}
  FunctionOutput(None=@n2.BaseColor) {outputName=OriginalBaseColor, sortPriority=2}
  n3 = FunctionInput() {inputName=Pre skinned Local Position, sortPriority=30}
  n4 = SetMaterialAttributes!(MaterialAttributes=@n2.MaterialAttributes, EmissiveColor=Constant()) {attributeSetTypes [#1]}
  n5 = FunctionInput() {inputName=Tech Art Mask, sortPriority=10}
  n6 = MaterialFunctionCall!(Material=@n4, TechArtMask=@n5) {materialFunction=&/MF_UV2_Composite_SL}
  n7 = StaticSwitchParameter(
   True=MaterialFunctionCall!(
    Material Attributes=@n6.Result
    M=FunctionInput() {inputName=M, inputType=FunctionInput_Scalar, sortPriority=20}
    Fresnel=FunctionInput() {inputName=Fresnel, inputType=FunctionInput_Scalar, sortPriority=40}
    TA Mask=@n5
    FX_Mask=@n6.Result
   ) {materialFunction=/BRCosmetics/Characters/Player/Male/Medium/Bodies/M_MED_LazyLizz_Chip/Materials/MF_LazyLizz_Chip_Skin_SL}.Result
   False=@n4
  ) {group=Superlevel, parameterName=UseLazyLizzSkin}
  n8 = MaterialFunctionCall!(PreskinnedPos=@n3, Material=@n7) {materialFunction=&/MF_MistRavenCowl_Emit_SL}
  n9 = MaterialFunctionCall!(
   EmissiveBlueChannel=@n8.EmissiveBlueChannel
   EmissiveTextureSampler=@n8.EmissiveBlueChannel
   Material=@n8.EmissiveBlueChannel
   PreSkinnedLocation=@n3
   CowlMask=@n8.EmissiveBlueChannel
   BodyMask=@n8.EmissiveBlueChannel
  ) {materialFunction=&/MF_MistravenCowl_Trans_SL}
  FunctionOutput(None=StaticSwitchParameter(True=@n9.Result, False=@n7) {parameterName=UseMistRavenSkin})
  FunctionOutput(None=@n2.EmissiveColor) {outputName=OriginalEmissive, sortPriority=1}
 ]}
