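# Material function that composites a configurable player banner onto the incoming material attributes, in the Ashen Magnus body's materials.
# Banner and mask textures get mirror, offset, rotation and scale UV controls, then multiply into a bounds-clipped blend mask.
# A Gmap input can shade the banner color, while roughness, metallic and specular mix overrides with adjusted incoming values.
# Static switches provide bypass and debug view; the secondary-banner branch is left unconnected.
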MaterialFunction /BRCosmetics/Characters/Player/Male/Medium/Bodies/M_MED_AshenMagnus_Slant/Materials/MF_AshenMagnus_PlayerBanner
graph {nodes [
  n1 = Multiply(A=NamedRerouteUsage(), B=Constant3Vector() {constant=<rgba 1 0 0 0>})
  n2 = Multiply(A=NamedRerouteUsage(), B=Constant3Vector() {constant=<rgba 0 1 0 0>})
  n3 = SetMaterialAttributes!(MaterialAttributes, BaseColor=LinearInterpolate(A=Max(A=@n1, B=@n2), B=NamedRerouteUsage()) {constAlpha=0.75}) {attributeSetTypes [#1]}
  n4 = GetMaterialAttributes(None=NamedRerouteDeclaration(None=FunctionInput() {inputName=Materials, inputType=FunctionInput_MaterialAttributes}) {name=IN_Material, nodeColor=<rgba 0.8047 1 0 1>}) {attributeGetTypes [
    #1
    #2
    #3
    #4
   ]}
  n5 = SetMaterialAttributes!(
   MaterialAttributes=@n4.MaterialAttributes
   BaseColor=NamedRerouteUsage()
   Metallic=NamedRerouteUsage()
   Specular=NamedRerouteUsage()
   Roughness=NamedRerouteUsage()
  ) {attributeSetTypes [
    #1
    #2
    #3
    #4
   ]}
  n6 = BlendMaterialAttributes(A=@n4.MaterialAttributes, B=@n5, Alpha=NamedRerouteUsage())
  FunctionOutput(None=StaticSwitchParameter(True=StaticSwitchParameter(True=@n3, False=@n6) {group=PlayerBanner, parameterName=DebugBanner, sortPriority=64}, False=NamedRerouteUsage()) {defaultValue=true, group=PlayerBanner, parameterName=UsePlayerBanner, sortPriority=0})
  n7 = Multiply(
   A=TextureCoordinate()
   B=AppendVector(
    A=ScalarParameter() {defaultValue=1, group=PlayerBanner, parameterName=Banner_MirrorX, sliderMax=1, sliderMin=-1, sortPriority=5}
    B=ScalarParameter() {defaultValue=1, group=PlayerBanner, parameterName=Banner_MirrorY, sliderMax=1, sliderMin=-1, sortPriority=5}
   )
  )
  n8 = AppendVector(
   A=ScalarParameter() {group=PlayerBanner, parameterName=Banner_X, sortPriority=6}
   B=ScalarParameter() {group=PlayerBanner, parameterName=Banner_Y, sortPriority=6}
  )
  n9 = MaterialFunctionCall(
   UVs=MaterialFunctionCall(UVs=Add(A=@n7, B=@n8), "Rotation Angle (0-1)"=ScalarParameter() {group=PlayerBanner, parameterName=Banner_Rotation, sortPriority=7}) {materialFunction=&CustomRotator}.Rotated Values
   Texture Scale=ScalarParameter() {defaultValue=0.15, group=PlayerBanner, parameterName=Banner_Scale, sortPriority=8}
  ) {materialFunction=&ScaleUVsByCenter}
  n10 = TextureSampleParameter2D(UVs=@n9.UVs) {
   group=PlayerBanner
   parameterName=Banner_Texture
   samplerSource=SSM_Clamp_WorldGroupSettings
   sortPriority=2
   texture=/Game/UI/Foundation/Textures/Banner/Shapes/banner_shape_standard_Mipmaps
  }
  n11 = Multiply(A=Multiply(A=@n10.RGB, B=@n10.RGB), B=@n9.0-1 mask)
  n12 = Multiply(
   A=TextureCoordinate()
   B=AppendVector(
    A=ScalarParameter() {defaultValue=1, group=PlayerBanner, parameterName=Mask_MirrorX, sliderMax=1, sliderMin=-1, sortPriority=10}
    B=ScalarParameter() {defaultValue=1, group=PlayerBanner, parameterName=Mask_MirrorY, sliderMax=1, sliderMin=-1, sortPriority=10}
   )
  )
  n13 = AppendVector(
   A=ScalarParameter() {group=PlayerBanner, parameterName=Mask_X, sortPriority=11}
   B=ScalarParameter() {group=PlayerBanner, parameterName=Mask_Y, sortPriority=11}
  )
  n14 = MaterialFunctionCall(
   UVs=MaterialFunctionCall(UVs=Add(A=@n12, B=@n13), "Rotation Angle (0-1)"=ScalarParameter() {group=PlayerBanner, parameterName=Mask_Rotation, sortPriority=12}) {materialFunction=&CustomRotator}.Rotated Values
   Texture Scale=ScalarParameter() {defaultValue=0.15, group=PlayerBanner, parameterName=Mask_Scale, sortPriority=13}
  ) {materialFunction=&ScaleUVsByCenter}
  n15 = TextureSampleParameter2D(UVs=StaticSwitchParameter(True=@n9.UVs, False=@n14.UVs) {defaultValue=true, group=PlayerBanner, parameterName=UseBannerUVsForMask, sortPriority=9}) {
   group=PlayerBanner
   parameterName=Banner_Mask
   samplerSource=SSM_Clamp_WorldGroupSettings
   sortPriority=3
   texture=&banner_shape_standard_Mipmaps
  }
  n16 = Multiply(
   A=StaticSwitchParameter(True=@n9.0-1 mask, False=@n14.0-1 mask) {defaultValue=true, group=PlayerBanner, parameterName=UseBannerUVsForMask, sortPriority=9}
   B=StaticComponentMaskParameter(None=@n15.RGBA) {defaultR=true, group=PlayerBanner, parameterName=Banner_MaskChannel, sortPriority=4}
  )
  n17 = LinearInterpolate(A=@n16, B=OneMinus(None=@n16), Alpha=ScalarParameter() {group=PlayerBanner, parameterName=Banner_MaskInvert, sortPriority=14})
  n18 = ScalarParameter() {group=PlayerBanner, parameterName=Banner_MaskBias, sortPriority=15}
  n19 = Saturate(None=Multiply(A=Add(A=@n17, B=@n18), B=ScalarParameter() {defaultValue=1, group=PlayerBanner, parameterName=Banner_MaskMult, sortPriority=16}))
  NamedRerouteDeclaration(None=Multiply(
    A=NamedRerouteDeclaration(None=@n11) {name=Banner_Texture, nodeColor=<rgba 1 0 0.8906 1>}
    B=NamedRerouteDeclaration(None=@n19) {name=Mask_Texture, nodeColor=<rgba 0 1 0.7891 1>}
   )) {name=PrimaryBanner_Mask, nodeColor=<rgba 1 0 0.0703 1>}
  NamedRerouteDeclaration(None=@n4.BaseColor) {name=IN_BaseColor, nodeColor=<rgba 1 0 0.2344 1>}
  NamedRerouteDeclaration(None=@n4.Metallic) {name=IN_Metallic, nodeColor=<rgba 0 0.5547 1 1>}
  NamedRerouteDeclaration(None=@n4.Specular) {name=IN_Specular, nodeColor=<rgba 0.875 1 0 1>}
  NamedRerouteDeclaration(None=@n4.Roughness) {name=IN_Roughness, nodeColor=<rgba 0.8047 0 1 1>}
  n20 = VectorParameter() {group=PlayerBanner, parameterName=Banner_Color, sortPriority=20, defaultValue=<rgba 1 0.1939 0.1354 1>}
  n21 = ScalarParameter() {group=PlayerBanner, parameterName=Banner_ColorGmapDark, sortPriority=23}
  n22 = ScalarParameter() {defaultValue=1, group=PlayerBanner, parameterName=Banner_ColorGmapLight, sortPriority=23}
  n23 = ChannelMaskParameter(None=FunctionInput() {bUsePreviewValueAsDefault=true, inputName=Gmap, previewValue=<1 1 1 1>}) {group=PlayerBanner, maskChannel=Green, parameterName=GMapChannel, sortPriority=22, defaultValue=<rgba 0 1 0 0>}
  n24 = Multiply(A=@n20.RGB, B=Saturate(None=LinearInterpolate(A=@n21, B=@n22, Alpha=@n23.R)))
  n25 = ScalarParameter() {defaultValue=1, group=PlayerBanner, parameterName=Banner_UseGmapToColor, sortPriority=21}
  NamedRerouteDeclaration(None=Saturate(None=LinearInterpolate(A=@n20.RGB, B=@n24, Alpha=@n25))) {name=OUT_BaseColor, nodeColor=<rgba 0 1 0.8828 1>}
  n26 = ScalarParameter() {defaultValue=0.5, group=PlayerBanner, parameterName=Banner_Roughness, sortPriority=24}
  n27 = Multiply(
   A=Add(A=NamedRerouteUsage(), B=ScalarParameter() {group=PlayerBanner, parameterName=Banner_RoughnessBias, sortPriority=26})
   B=ScalarParameter() {defaultValue=1, group=PlayerBanner, parameterName=Banner_RoughnessMult, sortPriority=27}
  )
  n28 = ScalarParameter() {defaultValue=1, group=PlayerBanner, parameterName=Banner_UseIncomingRoughness, sortPriority=25}
  NamedRerouteDeclaration(None=Saturate(None=LinearInterpolate(A=@n26, B=@n27, Alpha=@n28))) {name=OUT_Roughness, nodeColor=<rgba 0 1 0.7188 1>}
  n29 = ScalarParameter() {group=PlayerBanner, parameterName=Banner_Metallic, sortPriority=30}
  n30 = Multiply(
   A=Add(A=NamedRerouteUsage(), B=ScalarParameter() {group=PlayerBanner, parameterName=Banner_MetallicBias})
   B=ScalarParameter() {defaultValue=1, group=PlayerBanner, parameterName=Banner_MetallicMult, sortPriority=33}
  )
  n31 = ScalarParameter() {group=PlayerBanner, parameterName=Banner_UseIncomingMetallic, sortPriority=31}
  NamedRerouteDeclaration(None=Saturate(None=LinearInterpolate(A=@n29, B=@n30, Alpha=@n31))) {name=OUT_Metallic, nodeColor=<rgba 0.8984 0 1 1>}
  n32 = ScalarParameter() {defaultValue=0.5, group=PlayerBanner, parameterName=Banner_Specular, sortPriority=30}
  n33 = Multiply(
   A=Add(A=NamedRerouteUsage(), B=ScalarParameter() {group=PlayerBanner, parameterName=Banner_SpecularBias})
   B=ScalarParameter() {defaultValue=1, group=PlayerBanner, parameterName=Banner_SpecularMult, sortPriority=33}
  )
  n34 = ScalarParameter() {defaultValue=1, group=PlayerBanner, parameterName=Banner_UseIncomingSpecular, sortPriority=31}
  NamedRerouteDeclaration(None=Saturate(None=LinearInterpolate(A=@n32, B=@n33, Alpha=@n34))) {name=OUT_Specular, nodeColor=<rgba 0.0781 0 1 1>}
  n35 = Multiply(
   A=AppendVector(
    A=ScalarParameter() {defaultValue=1, group=PlayerBanner, parameterName=Banner2_MirrorX, sliderMax=1, sliderMin=-1}
    B=ScalarParameter() {defaultValue=1, group=PlayerBanner, parameterName=Banner2_MirrorY, sliderMax=1, sliderMin=-1}
   )
   B=TextureCoordinate()
  )
  n36 = AppendVector(
   A=ScalarParameter() {defaultValue=0.25, group=PlayerBanner, parameterName=Banner2_X}
   B=ScalarParameter() {group=PlayerBanner, parameterName=Banner2_Y}
  )
  n37 = MaterialFunctionCall(
   UVs=MaterialFunctionCall(UVs=Add(A=@n35, B=@n36), "Rotation Angle (0-1)"=ScalarParameter() {group=PlayerBanner, parameterName=Banner2_Rotation}) {materialFunction=&CustomRotator}.Rotated Values
   Texture Scale=ScalarParameter() {defaultValue=0.15, group=PlayerBanner, parameterName=Banner2_Scale}
  ) {materialFunction=&ScaleUVsByCenter}
  n38 = TextureSampleParameter2D(UVs=@n37.UVs) {
   group=PlayerBanner
   parameterName=Banner_Texture
   samplerSource=SSM_Clamp_WorldGroupSettings
   sortPriority=2
   texture=&banner_shape_standard_Mipmaps
  }
  StaticSwitchParameter(True=Multiply(A=@n38.R, B=@n37.0-1 mask), False=Constant()) {group=PlayerBanner, parameterName=UseSecondaryBanner, sortPriority=1}
  n39 = TextureSampleParameter2D(UVs=LinearInterpolate(A=@n37.UVs, B=TextureCoordinate(), Alpha=ScalarParameter() {group=PlayerBanner, parameterName=Banner1_UseUV0forBackgroundTexture})) {
   group=PlayerBanner
   parameterName=Banner_Background_Texture
   samplerSource=SSM_Clamp_WorldGroupSettings
   sortPriority=3
   texture=&banner_shape_standard_Mipmaps
  }
  StaticSwitchParameter(
   True=Multiply(
    A=@n37.0-1 mask
    B=StaticComponentMaskParameter(None=@n39.RGBA) {defaultR=true, group=PlayerBanner, parameterName=Banner1_BackgroundTextureChannel, sortPriority=4}
   )
   False=Constant()
  ) {group=PlayerBanner, parameterName=UseSecondaryBanner, sortPriority=1}
  VectorParameter() {group=PlayerBanner, parameterName=Banner_Color, sortPriority=20, defaultValue=<rgba 1 0.1939 0.1354 1>}
 ]}
