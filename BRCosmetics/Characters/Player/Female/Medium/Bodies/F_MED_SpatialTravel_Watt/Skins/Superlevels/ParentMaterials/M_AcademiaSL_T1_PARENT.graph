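# Parent character material for Academia T1 superlevel skins, wrapped by shared PreFX and PostFX functions.
# Mask textures layer metal, black paint, secondary color, effects and emissive outlines, with switches for AshenMagnus chest, banner and decal variants.
# Chest core and lightning use polar noise on a camera-facing virtual plane, scaled by depth and centered on a Blueprint-fed or local position.
# Pre-skinned coordinates drive procedural sine-wave leg strings and hand/torso masks.
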Material /BRCosmetics/Characters/Player/Female/Medium/Bodies/F_MED_SpatialTravel_Watt/Skins/Superlevels/ParentMaterials/M_AcademiaSL_T1_PARENT
graph {
 properties {
  bUseMaterialAttributes=true
  bUsedWithMeshDeformer=true
  bUsedWithMorphTargets=true
  bUsedWithSkeletalMesh=true
  pixelDepthOffsetMode=PDOM_Legacy
  shadingModel=MSM_FromMaterialExpression
  usedShadingModels=Default Lit | Subsurface Profile
 }
 nodes [
  Constant2Vector() {g=130, r=130}
  Constant2Vector() {g=130, r=130}
  n1 = TextureCoordinate() {uTiling=3, vTiling=3}
  DDX(Value=@n1)
  n2 = TextureCoordinate() {uTiling=3, vTiling=3}
  DDX(Value=@n2)
  DDY(Value=@n1)
  DDY(Value=@n2)
  n3 = MaterialFunctionCall!(Overide Texture Set UVs, Override Tech Art Mask UVs, Override Basic Emissive UVs) {materialFunction=/BRCosmetics/Characters/Player/Female/Medium/Bodies/F_MED_SpatialTravel_Watt/Skins/Superlevels/ParentMaterials/PreFX/MF_AcademiaSL_PreFX}
  n4 = NamedRerouteDeclaration(None=@n3.Pre skinned Local Position) {name=PreskinnedLocalPosition, nodeColor=<rgba 0.5703 1 0 1>}
  MaterialFunctionCall(Material=NamedRerouteUsage(), M=@n3.M, Pre Skinned Local Position=@n4, Distance Blend=@n3.M, Fresnel=@n3.M) {materialFunction=&MF_CharacterParent_2023_PostFX}
  n5 = Add(A=Multiply(A=NamedRerouteUsage()) {constB=2}) {constB=-1}
  n6 = NamedRerouteUsage()
  n7 = Multiply(A=Multiply(A=Time()) {constB=10}, B=Constant3Vector() {constant=<rgba 0.15 -0.01 1 0>})
  n8 = MaterialFunctionCall!(
   Export Float 4
   TextureObject=TextureObject() {texture=/BRCosmetics/Characters/Player/Female/Medium/Bodies/F_MED_SpatialTravel_Watt/Skins/Superlevels/Textures/3d_VecNoise}
   TextureSize=Multiply(A=ScalarParameter() {defaultValue=1.56, parameterName=PanningNoiseSize}) {constB=64}
   WorldPosition=Add(A=Multiply(A=@n5) {constB=10}, B=Add(A=@n6, B=@n7))
   World Space Normal=NamedRerouteUsage()
   RotZAxis
   RotXAxis
   RotYAxis
   ProjectionTransitionContrast
  ) {materialFunction=&/MF_WorldAlignedTexture_SL}
  n9 = Add(A=Multiply(A=ComponentMask(None=@n8.XYZ Texture) {b=true}), B=ComponentMask(None=@n8.XYZ Texture) {g=true})
  n10 = Multiply(A=NamedRerouteDeclaration(None=Multiply(A=@n9) {constB=1.5}) {name=PanningNoise, nodeColor=<rgba 1 0.4219 0 1>}, B=NamedRerouteUsage())
  n11 = NamedRerouteUsage()
  Multiply(A=@n10, B=Saturate(None=Add(A=@n11) {constB=0.2}))
  Multiply(A=NamedRerouteUsage())
  n12 = MaterialFunctionCall!(
   AshenMagnus_OriginalEmissive=NamedRerouteUsage()
   Result=@n3.Result
   Tech Art Mask=@n3.Result
   M=NamedRerouteDeclaration(None=@n3.M) {name=InMask, nodeColor=<rgba 1 0.4922 0 1>}
   Pre skinned Local Position=@n4
   Fresnel=NamedRerouteDeclaration(None=@n3.Fresnel) {name=InFresnel, nodeColor=<rgba 0 0.25 1 1>}
  ) {materialFunction=&/MF_AcademiaSL_InitialSetup}
  NamedRerouteDeclaration(None=@n12.OriginalBaseColor) {name=OriginalBaseColor, nodeColor=<rgba 0 1 0.4844 1>}
  NamedRerouteDeclaration(None=@n12.Result) {name=OriginalMA, nodeColor=<rgba 1 0.8438 0 1>}
  n13 = NamedRerouteUsage()
  n14 = VectorParameter() {group=Superlevel, parameterName=Hand_Position, defaultValue=<rgba 50 5 88 1>}
  n15 = ScalarParameter() {defaultValue=35, group=Superlevel, parameterName=HandMask_Radius}
  n16 = Multiply(A=@n14.RGB, B=Constant3Vector() {constant=<rgba -1 1 1 0>})
  n17 = Multiply(
   A=Add(
    A=SphereMask(A=@n13, B=@n14.RGB, Radius=@n15) {attenuationRadius=30, hardnessPercent=0.1}
    B=SphereMask(A=@n13, B=@n16, Radius=@n15) {attenuationRadius=30, hardnessPercent=0.1}
   )
   B=ScalarParameter() {defaultValue=1, parameterName=HandMask_Intensity}
  )
  n18 = VectorParameter() {parameterName=Torso_Position, defaultValue=<rgba 0 0 100 1>}
  n19 = ScalarParameter() {defaultValue=40, group=Superlevel, parameterName=PanningNoise_TorsoRadius}
  n20 = Multiply(
   A=Power(Base=SphereMask(A=@n13, B=@n18.RGB, Radius=@n19) {attenuationRadius=40, hardnessPercent=0}) {constExponent=3}
   B=ScalarParameter() {defaultValue=0.6, group=Superlevel, parameterName=PanningNoise_TorsoIntensity}
  )
  NamedRerouteDeclaration(None=Multiply(A=Add(A=@n17, B=@n20)) {constB=0.5}) {name=TorsoMask, nodeColor=<rgba 1 0.7969 0 1>}
  n21 = NamedRerouteUsage()
  n22 = TransformPosition(None=NamedRerouteUsage()) {transformSourceType=TRANSFORMPOSSOURCE_World, transformType=TRANSFORMPOSSOURCE_View}
  n23 = NamedRerouteDeclaration(None=Divide(B=ComponentMask(None=@n22) {b=true})) {name=ChestPosition_ViewSpaceScale, nodeColor=<rgba 1 0.2714 0.0299 1>}
  n24 = Multiply(A=@n23) {constB=7.5}
  n25 = MaterialFunctionCall(UVs=NamedRerouteUsage(), Texture Scale=Multiply(A=@n23) {constB=30}) {materialFunction=&ScaleUVsByCenter}
  n26 = TextureSample(UVs=Panner(Coordinate=MaterialFunctionCall!(UV Coord=@n25.UVs, LenghtScale, RadialScale) {materialFunction=/Game/Global/MaterialFunctions/TechArt/MF_PolarCoordinates}.Result) {speedX=-0.35, speedY=0.01}) {
   samplerSource=SSM_Wrap_WorldGroupSettings
   texture=/BRCosmetics/Characters/Player/Female/Medium/Bodies/F_MED_SpatialTravel_Watt/Skins/Superlevels/Textures/T_AcademiaSL_T1_Masks
  }
  n27 = NamedRerouteDeclaration(None=@n26.B) {name=LightningNoise, nodeColor=<rgba 0.2422 1 0 1>}
  n28 = Add(A=Multiply(A=@n27) {constB=2}) {constB=-1}
  n29 = TextureSample(UVs=Add(
    A=MaterialFunctionCall(UVs=@n21, Texture Scale=@n24) {materialFunction=&ScaleUVsByCenter}.UVs
    B=Multiply(A=@n28) {constB=0.15}
   ) {constB=0}) {
   samplerSource=SSM_Clamp_WorldGroupSettings
   texture=&T_AcademiaSL_T1_Masks
  }
  n30 = NamedRerouteUsage()
  n31 = ScalarParameter() {defaultValue=30, group=Superlevel, parameterName=CoreEffect_Area}
  n32 = TextureSample(UVs=MaterialFunctionCall(UVs=@n30, Texture Scale=Multiply(A=@n23, B=@n31) {constB=40}) {materialFunction=&ScaleUVsByCenter}.UVs) {
   samplerSource=SSM_Clamp_WorldGroupSettings
   texture=&T_AcademiaSL_T1_Masks
  }
  n33 = Multiply(A=Multiply(A=@n29.A, B=ScalarParameter() {defaultValue=1, group=Superlevel, parameterName=Core_Brightness}), B=Multiply(A=@n32.G) {constB=2})
  n34 = Add(A=Time(), B=Multiply(A=@n32.G) {constB=2})
  n35 = Add(A=Multiply(A=Sine(None=@n34)) {constB=0.5}) {constB=0.5}
  n36 = Saturate(None=Multiply(A=@n32.G) {constB=2})
  n37 = ScalarParameter() {defaultValue=40, group=SUPERLEVEL, parameterName=ChestPattern_Scale, sortPriority=101}
  n38 = Rotator(
   Coordinate=MaterialFunctionCall(UVs=@n30, Texture Scale=Multiply(A=@n23, B=@n37)) {materialFunction=&ScaleUVsByCenter}.UVs
   Time=Multiply(A=Time()) {constB=-0.15}
  )
  n39 = Multiply(B=Add(A=Multiply(B=@n27) {constA=2, constB=2}) {constB=-1}) {constA=0.0065}
  n40 = Time()
  n41 = TextureSample(UVs=Rotator(Coordinate=Add(A=@n38, B=@n39), Time=Multiply(A=@n40) {constB=2})) {
   samplerSource=SSM_Clamp_WorldGroupSettings
   texture=&T_AcademiaSL_T1_Masks
  }
  n42 = Add(A=Multiply(A=Multiply(A=@n35, B=@n36)) {constB=0.03}, B=Multiply(A=Multiply(A=@n32.G) {constB=5}, B=Multiply(A=@n41.R) {constB=0.5}))
  n43 = NamedRerouteUsage()
  n44 = NamedRerouteUsage()
  n45 = WorldPosition()
  n46 = NamedRerouteUsage()
  n47 = ScalarParameter() {defaultValue=60, group=Superlevel, parameterName=SphereMask_Chest_Radius}
  n48 = Multiply(
   A=Add(A=Add(A=@n33, B=@n42), B=Multiply(A=@n43, B=@n44))
   B=NamedRerouteDeclaration(None=SphereMask(A=@n45.XYZ, B=@n46, Radius=@n47) {attenuationRadius=41, hardnessPercent=0.2}) {name=SphereMask, nodeColor=<rgba 0 1 0.4609 1>}
  )
  n49 = ScalarParameter() {defaultValue=0.8, group=Superlevel, parameterName=OverallBrightness}
  NamedRerouteDeclaration(None=Add(A=Multiply(A=@n48, B=@n49), B=NamedRerouteUsage())) {name=LightningEffect, nodeColor=<rgba 1 0.9141 0 1>}
  n50 = ComponentMask(None=NamedRerouteUsage()) {b=true}
  n51 = OneMinus(None=Saturate(None=Divide(A=@n50) {constB=180}))
  n52 = Multiply(A=Multiply(A=Constant3Vector() {constant=<rgba 0.1125 0 0.7031 1>}, B=Power(Base=@n51) {constExponent=23})) {constB=13}
  n53 = Constant3Vector() {constant=<rgba 0.7031 0.0657 0.4439 1>}
  n54 = Divide(A=@n50) {constB=100}
  n55 = Power(Base=@n51) {constExponent=7}
  n56 = Divide(A=NamedRerouteUsage()) {constB=100}
  n57 = TextureSample(UVs=Panner(Coordinate=ComponentMask(None=@n56) {b=true, r=true}) {speedX=0.05, speedY=0.6}) {
   samplerSource=SSM_Wrap_WorldGroupSettings
   texture=&3d_VecNoise
  }
  n58 = Saturate(None=Add(A=Multiply(A=@n57.R) {constB=0.5}, B=@n57.G))
  n59 = Multiply(A=Add(A=Multiply(A=@n58) {constB=2}) {constB=-1}) {constB=5}
  n60 = NamedRerouteUsage()
  n61 = Divide(A=ComponentMask(None=Add(A=@n59, B=@n60)) {g=true, r=true}, B=Constant() {r=100})
  n62 = MaterialFunctionCall(XY=Add(A=@n61, B=Constant2Vector()) {constB=0.5}) {materialFunction=&Swizzle}
  n63 = Abs(None=Subtract(A=Multiply(A=@n62.YX) {constB=2}) {constB=0.98})
  n64 = Frac(None=ComponentMask(None=Multiply(A=@n63) {constB=2}) {g=true})
  n65 = ScalarParameter() {parameterName=AAASsS}
  n66 = Multiply(A=ComponentMask(None=Add(A=@n61, B=@n65) {constB=0}) {g=true}) {constB=2.4}
  n67 = Saturate(None=OneMinus(None=@n66))
  n68 = Time()
  n69 = ScalarParameter() {defaultValue=0.3, group=TechArt, parameterName=StringBaseSineFrequency, sortPriority=40}
  n70 = Sine(None=Add(A=@n66, B=Multiply(A=@n68, B=@n69) {constB=0.3111}))
  n71 = Add(
   A=Multiply(A=Multiply(B=@n67) {constA=1}, B=ScalarParameter() {defaultValue=0.1, group=TechArt, parameterName=StringExciteAmt})
   B=Multiply(A=Multiply(A=@n67, B=@n70), B=ScalarParameter() {defaultValue=0.2, group=TechArt, parameterName=StringBaseSineAmplitude, sortPriority=40}) {constB=0.2}
  )
  n72 = ScalarParameter() {defaultValue=0.356, group=TechArt, parameterName=StringSeparation, sliderMax=0.5, sortPriority=38}
  n73 = Multiply(
   A=SphereMask(
    A=Add(A=@n64, B=@n71)
    B=ScalarParameter() {defaultValue=0.6, group=TechArt, parameterName=StringOffset, sortPriority=38}
    Radius=ScalarParameter() {defaultValue=0.3, group=TechArt, parameterName=StringThickness, sortPriority=38}
   ) {attenuationRadius=1, hardnessPercent=0}
   B=Add(A=@n72)
  )
  n74 = Subtract(A=Abs(None=OneMinus(None=@n73)), B=@n72) {constB=0.05}
  n75 = OneMinus(None=Divide(A=Abs(None=@n74), B=OneMinus(None=@n72)))
  n76 = ScalarParameter() {defaultValue=15, group=TechArt, parameterName=StringThicknessExp, sortPriority=38}
  n77 = NamedRerouteUsage()
  n78 = Divide(A=ComponentMask(None=Add(A=@n59, B=@n77)) {g=true, r=true}, B=Constant() {r=100})
  n79 = Constant2Vector()
  n80 = Multiply(A=MaterialFunctionCall(XY=Add(A=@n78, B=@n79) {constB=0.5}) {materialFunction=&Swizzle}.YX) {constB=2}
  n81 = Multiply(A=Abs(None=Subtract(A=@n80) {constB=0.98})) {constB=2}
  n82 = ScalarParameter() {parameterName=AAASsS}
  n83 = Multiply(A=ComponentMask(None=Add(A=@n78, B=@n82) {constB=0}) {g=true}) {constB=2.4}
  n84 = Saturate(None=OneMinus(None=@n83))
  n85 = Multiply(B=@n84) {constA=1}
  n86 = ScalarParameter() {defaultValue=0.1, group=TechArt, parameterName=StringExciteAmt}
  n87 = Multiply(A=Time(), B=ScalarParameter() {defaultValue=0.09, group=TechArt, parameterName=StringBaseSineFrequency_01, sortPriority=40}) {constB=0.3111}
  n88 = Multiply(A=@n84, B=Sine(None=Add(A=@n83, B=@n87)))
  n89 = ScalarParameter() {defaultValue=0.2, group=TechArt, parameterName=StringBaseSineAmplitude, sortPriority=40}
  n90 = Add(A=Frac(None=ComponentMask(None=@n81) {g=true}), B=Add(A=Multiply(A=@n85, B=@n86), B=Multiply(A=@n88, B=@n89) {constB=0.2}))
  n91 = ScalarParameter() {defaultValue=0.6, group=TechArt, parameterName=StringOffset_01, sortPriority=38}
  n92 = ScalarParameter() {defaultValue=0.2, group=TechArt, parameterName=StringThickness_01, sortPriority=38}
  n93 = ScalarParameter() {defaultValue=0.356, group=TechArt, parameterName=StringSeparation, sliderMax=0.5, sortPriority=38}
  n94 = OneMinus(None=Multiply(A=SphereMask(A=@n90, B=@n91, Radius=@n92) {attenuationRadius=1, hardnessPercent=0}, B=Add(A=@n93)))
  n95 = Abs(None=Subtract(A=Abs(None=@n94), B=@n93) {constB=0.05})
  n96 = OneMinus(None=@n93)
  n97 = Power(
   Base=OneMinus(None=Divide(A=@n95, B=@n96))
   Exp=ScalarParameter() {defaultValue=10, group=TechArt, parameterName=StringThicknessExp_01, sortPriority=38}
  ) {constExponent=10}
  n98 = NamedRerouteUsage()
  n99 = Divide(A=ComponentMask(None=Add(A=@n59, B=@n98)) {g=true, r=true}, B=Constant() {r=100})
  n100 = Add(A=@n99, B=Constant2Vector()) {constB=0.5}
  n101 = Subtract(A=Multiply(A=MaterialFunctionCall(XY=@n100) {materialFunction=&Swizzle}.YX) {constB=2}) {constB=0.98}
  n102 = ComponentMask(None=Multiply(A=Abs(None=@n101)) {constB=2}) {g=true}
  n103 = ScalarParameter() {parameterName=AAASsS}
  n104 = Multiply(A=ComponentMask(None=Add(A=@n99, B=@n103) {constB=0}) {g=true}) {constB=2.4}
  n105 = Saturate(None=OneMinus(None=@n104))
  n106 = Multiply(A=Multiply(B=@n105) {constA=1}, B=ScalarParameter() {defaultValue=0.2, group=TechArt, parameterName=StringExciteAmt_02})
  n107 = Add(
   A=@n104
   B=Multiply(A=Time(), B=ScalarParameter() {defaultValue=0.11, group=TechArt, parameterName=StringBaseSineFrequency_02, sortPriority=40}) {constB=0.3111}
  )
  n108 = Multiply(
   A=Multiply(A=@n105, B=Sine(None=@n107))
   B=ScalarParameter() {defaultValue=0.2, group=TechArt, parameterName=StringBaseSineAmplitude, sortPriority=40}
  ) {constB=0.2}
  n109 = SphereMask(
   A=Add(A=Frac(None=@n102), B=Add(A=@n106, B=@n108))
   B=ScalarParameter() {defaultValue=0.2, group=TechArt, parameterName=StringOffset_02, sortPriority=38}
   Radius=ScalarParameter() {defaultValue=0.2, group=TechArt, parameterName=StringThickness_02, sortPriority=38}
  ) {attenuationRadius=1, hardnessPercent=0}
  n110 = ScalarParameter() {defaultValue=0.356, group=TechArt, parameterName=StringSeparation, sliderMax=0.5, sortPriority=38}
  n111 = Add(A=@n110)
  n112 = Abs(None=OneMinus(None=Multiply(A=@n109, B=@n111)))
  n113 = Divide(A=Abs(None=Subtract(A=@n112, B=@n110) {constB=0.05}), B=OneMinus(None=@n110))
  n114 = Saturate(None=Power(Base=OneMinus(None=@n113), Exp=ScalarParameter() {defaultValue=12, group=TechArt, parameterName=StringThicknessExp_02, sortPriority=38}) {constExponent=10})
  n115 = Add(
   A=Multiply(A=Power(Base=@n75, Exp=@n76) {constExponent=10}) {constB=10}
   B=Add(A=Multiply(A=@n97) {constB=7}, B=Multiply(A=@n114) {constB=10})
  )
  n116 = PreSkinnedNormal()
  n117 = VertexInterpolator(VS=@n116)
  n118 = ComponentMask(None=NamedRerouteDeclaration(None=ShaderStageSwitch(PixelShader=@n117.PS, VertexShader=@n116)) {name=PreskinnedLocalNormal, nodeColor=<rgba 1 0.2578 0 1>}) {g=true}
  n119 = Power(Base=Saturate(None=Abs(None=@n118))) {constExponent=8}
  n120 = Multiply(
   A=LinearInterpolate(A=Constant3Vector() {constant=<rgba 0.1807 0.0006 0.7031 1>}, B=@n53, Alpha=Saturate(None=@n54))
   B=Multiply(A=Multiply(A=@n55) {constB=2}, B=Multiply(A=@n115, B=@n119))
  )
  n121 = Divide(A=@n50) {constB=7}
  n122 = Multiply(A=OneMinus(None=Saturate(None=@n121)), B=Multiply(A=@n53) {constB=3})
  n123 = OneMinus(None=NamedRerouteUsage())
  n124 = Divide(A=@n50) {constB=10}
  n125 = Multiply(A=OneMinus(None=Saturate(None=@n124))) {constB=20}
  NamedRerouteDeclaration(None=LinearInterpolate(A=Add(A=@n52, B=@n120), B=Multiply(A=@n122, B=@n123), Alpha=Saturate(None=@n125)) {constB=0}) {name=StringsLegs, nodeColor=<rgba 1 0.7266 0 1>}
  n126 = TextureSample() {
   samplerSource=SSM_Wrap_WorldGroupSettings
   texture=/BRCosmetics/Characters/Player/Male/Medium/Bodies/M_MED_AshenMagnus_Slant/Skins/Rye/Textures/T_AshenMagnus_SlantRye_Body_E
  }
  n127 = NamedRerouteDeclaration(None=@n126.RGB) {name=AshenMagnus_OriginalEmissive, nodeColor=<rgba 0 0.9297 1 1>}
  n128 = MaterialFunctionCall(Float3=Saturate(None=Multiply(A=@n127) {constB=2})) {materialFunction=&BreakOutFloat3Components}
  n129 = Multiply(
   A=Max(A=Max(A=@n128.R, B=@n128.R), B=@n128.B)
   B=MaterialFunctionCall!(A=TextureCoordinate(), B=Constant2Vector() {g=0.215, r=0.365}, Bounds=Constant2Vector() {g=0.1, r=0.1}, Edge Falloff=Constant()) {materialFunction=/Engine/Functions/Engine_MaterialFunctions01/Gradient/BoxMask-2D}.Result
  )
  NamedRerouteDeclaration(None=Saturate(None=Multiply(A=@n129) {constB=2})) {name=AshenMagnus_ChestMask, nodeColor=<rgba 1 0 0.5391 1>}
  NamedRerouteDeclaration(None=@n12.OriginalEmissive) {name=OriginalEmissive, nodeColor=<rgba 0.1484 0 1 1>}
  n130 = NamedRerouteUsage()
  n131 = NamedRerouteUsage()
  n132 = VectorParameter() {group=Superlevel, parameterName=Metal_Color, sortPriority=30, defaultValue=<rgba 0.2622 0.2622 0.2622 0>}
  n133 = BlendMaterialAttributes(
   A=@n131
   B=SetMaterialAttributes!(MaterialAttributes=@n131, Base Color=@n132.RGB, Metallic=Constant() {r=1}) {attributeSetTypes [#1, #2]}
   Alpha=ComponentMask(None=NamedRerouteUsage()) {g=true}
  )
  n134 = Constant() {r=0.05}
  n135 = Constant() {r=0.325}
  n136 = Constant()
  n137 = TextureSampleParameter2D() {
   group=Superlevel
   parameterName=SL_Masks_01
   samplerSource=SSM_Wrap_WorldGroupSettings
   samplerType=SAMPLERTYPE_Masks
   sortPriority=0
   texture=/BRCosmetics/Characters/Player/Female/Medium/Bodies/F_MED_SpatialTravel_Quirk/Textures/T_SpatialTravel_Quirk_Body_SL_FX
  }
  n138 = Add(A=Add(A=@n137.R, B=@n137.R), B=@n137.B)
  n139 = OneMinus(None=Saturate(None=StaticSwitchParameter(True=@n136, False=@n138) {group=Superlevel, parameterName=UseBlackPaintOnly}))
  n140 = ComponentMask(None=NamedRerouteUsage()) {b=true}
  n141 = GetMaterialAttributes(None=BlendMaterialAttributes(
    A=@n133
    B=SetMaterialAttributes!(MaterialAttributes=@n133, Base Color=@n134, Roughness=@n135) {attributeSetTypes [#1, #3]}
    Alpha=Subtract(A=@n139, B=@n140)
   )) {attributeGetTypes [#4]}
  n142 = MaterialFunctionCall() {materialFunction=&BoundingBoxBased_0-1_UVW}
  n143 = NamedRerouteUsage()
  n144 = LinearInterpolate(
   A=Constant3Vector() {constant=<rgba 0.6265 0.1156 1 1>}
   B=Constant3Vector() {constant=<rgba 0.1524 0 1 1>}
   Alpha=Saturate(None=Multiply(A=@n143) {constB=2})
  )
  n145 = Multiply(A=Multiply(A=Power(Base=@n142.B) {constExponent=5}, B=@n144)) {constB=0.3}
  n146 = NamedRerouteUsage()
  n147 = Constant3Vector() {constant=<rgba 0.9854 0.1107 1 1>}
  n148 = Constant3Vector() {constant=<rgba 0.3651 0 1 1>}
  n149 = Saturate(None=Multiply(A=NamedRerouteUsage()) {constB=10})
  n150 = NamedRerouteUsage()
  n151 = Multiply(A=Multiply(A=Power(Base=@n150)) {constB=2}, B=NamedRerouteUsage())
  n152 = NamedRerouteUsage()
  n153 = Add(A=Multiply(A=Multiply(A=@n146) {constB=2}, B=LinearInterpolate(A=@n147, B=@n148, Alpha=@n149)), B=Multiply(A=@n144, B=Add(A=@n151, B=@n152)))
  n154 = NamedRerouteUsage()
  n155 = BlendMaterialAttributes(
   A=@n141.MaterialAttributes
   B=SetMaterialAttributes!(MaterialAttributes=@n141.MaterialAttributes, Base Color=Max(A=@n145) {constB=0.01}, Emissive Color=Add(A=@n153, B=@n154), Normal) {attributeSetTypes [#1, #5, #4]}
   Alpha=ComponentMask(None=NamedRerouteUsage()) {r=true}
  )
  n156 = Constant() {r=0.5}
  n157 = VectorParameter() {group=Superlevel, parameterName=Secondary_Color, sortPriority=30, defaultValue=<rgba 1 0.1084 0.7374 1>}
  n158 = Constant() {r=0.5}
  n159 = NamedRerouteUsage()
  n160 = GetMaterialAttributes(None=BlendMaterialAttributes(
    A=@n155
    B=SetMaterialAttributes!(MaterialAttributes=@n155, Roughness=@n156, Base Color=@n157.RGB, Metallic=@n158) {attributeSetTypes [#3, #1, #2]}
    Alpha=ComponentMask(None=@n159) {b=true}
   )) {attributeGetTypes [#5]}
  n161 = VectorParameter() {group=Superlevel, parameterName=EmissiveDefaultVariant_Color, sortPriority=30, defaultValue=<rgba 6 0.2164 1.608 1>}
  n162 = Multiply(A=@n161.RGB, B=OneMinus(None=NamedRerouteUsage()))
  n163 = Multiply(A=@n162) {constB=0.7}
  n164 = VectorParameter() {group=Superlevel, parameterName=Outlines_Color, sortPriority=30, defaultValue=<rgba 1 0.0665 0.5571 1>}
  n165 = Multiply(A=@n164.RGB) {constB=3}
  n166 = TextureSampleParameter2D() {
   group=Superlevel
   parameterName=SL_Masks_02
   samplerSource=SSM_Wrap_WorldGroupSettings
   samplerType=SAMPLERTYPE_Masks
   sortPriority=0
   texture=/BRCosmetics/Characters/Player/Female/Medium/Bodies/F_MED_SpatialTravel_Quirk/Textures/T_SpatialTravel_Quirk_Body_SL_FX2
  }
  n167 = StaticSwitchParameter(True=Constant(), False=ComponentMask(None=@n166.RGB) {g=true, r=true}) {group=Superlevel, parameterName=UseBlackPaintOnly}
  n168 = TextureCoordinate()
  n169 = Constant() {r=0.25}
  n170 = Constant() {r=0.05}
  n171 = Constant() {r=0.4}
  n172 = Constant() {r=0.59}
  n173 = SphereMask(A=NamedRerouteUsage(), B=Constant3Vector() {constant=<rgba 0.6584 0.6584 0.6584 1>}) {attenuationRadius=0.4, hardnessPercent=1}
  n174 = StaticSwitchParameter(
   True=Multiply(
    A=MaterialFunctionCall!(
     Texture Coordinate=@n168
     X Pos Right=@n169
     X Pos Left=@n170
     Y Pos Up=@n171
     Y Pos Down=@n172
     SquareSoftness
     RotateSquare
     RotationAmount
    ) {materialFunction=/Game/Accessories/FORT_Backpacks/Backpack_F_MED_GroovyReader_Play/Materials/MaterialFunctions/MF_SquareMaskGenerator}.SquareMask
    B=Multiply(A=@n173) {constB=1.5}
   )
   False=Constant()
  ) {group=Superlevel, parameterName=UseLazyLizzChipDecal}
  n175 = Add(A=ComponentMask(None=@n167) {g=true}, B=@n174)
  n176 = ComponentMask(None=@n167) {r=true}
  n177 = LinearInterpolate(
   A=LinearInterpolate(
    A=@n160.Emissive Color
    B=StaticSwitchParameter(True=@n163, False=@n165) {group=Superlevel, parameterName=UsePolarityWinnFaceAcc}
    Alpha=Saturate(None=@n175)
   )
   B=@n162
   Alpha=Saturate(None=Add(A=@n176, B=@n174))
  )
  n178 = NamedRerouteUsage()
  n179 = Constant() {r=1}
  n180 = StaticSwitchParameter(
   True=MaterialFunctionCall(Hue Shift Percentage=ScalarParameter() {parameterName=AshenMagnusChest_HueShift}, Texture=@n178) {materialFunction=&HueShift}.Result
   False=Multiply(
    A=Desaturation(None=@n178, Fraction=@n179)
    B=VectorParameter() {group=Superlevel, parameterName=AshenMagnus_ChestTint, defaultValue=<rgba 1 1 1 0>}.RGB
   )
  ) {group=Superlevel, parameterName=AshenMagnus_UseHueShiftOnChest}
  n181 = NamedRerouteUsage()
  n182 = SetMaterialAttributes!(
   MaterialAttributes=@n160.MaterialAttributes
   Emissive Color=StaticSwitchParameter(True=LinearInterpolate(A=@n177, B=@n180, Alpha=@n181), False=@n177) {group=Superlevel, parameterName=UseAshenMagnusChest}
  ) {attributeSetTypes [#5]}
  NamedRerouteDeclaration(None=StaticSwitchParameter(
    True=MaterialFunctionCall!(Gmap=@n130, Materials=@n182) {materialFunction=/BRCosmetics/Characters/Player/Male/Medium/Bodies/M_MED_AshenMagnus_Slant/Materials/MF_AshenMagnus_PlayerBanner}.Result
    False=@n182
   ) {group=Superlevel, parameterName=UseAshenMagnusBanner}) {name=OutputMA, nodeColor=<rgba 0 1 0.4141 1>}
  n183 = NamedRerouteUsage()
  n184 = ScalarParameter() {defaultValue=60, group=Superlevel, parameterName=LightningRange_Area}
  n185 = TextureSample(UVs=MaterialFunctionCall(UVs=NamedRerouteUsage(), Texture Scale=Multiply(A=@n183, B=@n184) {constB=28}) {materialFunction=&ScaleUVsByCenter}.UVs) {
   samplerSource=SSM_Clamp_WorldGroupSettings
   texture=&T_AcademiaSL_T1_Masks
  }
  NamedRerouteDeclaration(None=Multiply(A=Power(Base=@n185.G)) {constB=5}) {name=LightningEffectRadius, nodeColor=<rgba 1 0.0938 0 1>}
  n186 = MaterialFunctionCall() {materialFunction=&CameraDirectionVector}
  n187 = Multiply(A=Constant3Vector() {constant=<rgba 1 1 0 0>}, B=@n186.Result)
  n188 = Add(A=Multiply(B=Normalize(VectorInput=@n187)) {constA=0.25}, B=@n186.Result)
  n189 = VectorParameter() {group=Superlevel, parameterName=CorePosition_Override, sortPriority=600000, defaultValue=<rgba 0 0 125 0>}
  n190 = VectorParameter() {group=Superlevel, parameterName=LocalChestPosition, sortPriority=1000, defaultValue=<rgba 0 0 125 0>}
  n191 = TransformPosition(None=Add(A=@n190.RGB, B=VectorParameter() {group=Superlevel, parameterName=PreviewOffset_Chest, sortPriority=800}.RGB)) {transformType=TRANSFORMPOSSOURCE_World}
  n192 = VectorParameter() {group=Superlevel, parameterName=BP_OffsetVector_Chest, sortPriority=1000}
  n193 = Add(
   A=Multiply(
    A=ScalarParameter() {defaultValue=-3, group=SUPERLEVEL, parameterName=PositionOffset_Chest, sortPriority=150}
    B=Normalize(VectorInput=@n192.RGB)
   )
   B=VectorParameter() {group=Superlevel, parameterName=BP_ChestPosition, sortPriority=1000}.RGB
  )
  n194 = ScalarParameter() {group=SUPERLEVEL, parameterName=BP_UseCCPM, sortPriority=0}
  n195 = NamedRerouteDeclaration(None=StaticSwitchParameter(True=TransformPosition(None=@n189.RGB) {transformType=TRANSFORMPOSSOURCE_World}, False=LinearInterpolate(A=@n191, B=@n193, Alpha=@n194)) {group=SUPERLEVEL, parameterName=OverrideCorePosition, sortPriority=600000}) {name=ChestPosition, nodeColor=<rgba 0 1 0.2734 1>}
  n196 = Constant() {r=45}
  n197 = Constant() {r=90}
  n198 = MaterialFunctionCall() {materialFunction=/Engine/Functions/Engine_MaterialFunctions02/Utility/FOV}
  n199 = Multiply(
   A=LinearInterpolate(A=Constant() {r=64}, B=Constant() {r=128}, Alpha=InverseLinearInterpolate(A=@n196, B=@n197, Value=@n198.FOV angle in degrees))
   B=Distance(A=@n195, B=CameraPositionWS())
  )
  NamedRerouteDeclaration(None=MaterialFunctionCall(Plane Normal Axis=Normalize(VectorInput=@n188), Plane Center=Add(A=@n186.Result, B=@n195), UV Scale=Multiply(A=@n199) {constB=0.0146}) {materialFunction=&VirtualPlaneCoordinates}.UVs) {name=ChestAllignedScreenPosition, nodeColor=<rgba 0.125 1 0 1>}
  n200 = NamedRerouteUsage()
  n201 = VectorParameter() {parameterName=C, defaultValue=<rgba 0 0 80 0>}
  n202 = ScalarParameter() {defaultValue=40, parameterName=CrotchMask_Radius}
  NamedRerouteDeclaration(None=OneMinus(None=SphereMask(A=@n200, B=@n201.RGB, Radius=@n202) {attenuationRadius=0, hardnessPercent=0.2})) {name=CrotchMask, nodeColor=<rgba 1 0 0.0469 1>}
  NamedRerouteDeclaration(None=@n10) {name=AdditionalRays, nodeColor=<rgba 0.1719 1 0 1>}
  NamedRerouteDeclaration(None=StaticSwitchParameter(True=@n136, False=@n137.RGB) {group=Superlevel, parameterName=UseBlackPaintOnly}) {name=SL_Masks, nodeColor=<rgba 1 0.6094 0 1>}
  n203 = NamedRerouteUsage()
  n204 = NamedRerouteUsage()
  n205 = Divide(B=ComponentMask(None=TransformPosition(None=@n204) {transformSourceType=TRANSFORMPOSSOURCE_World, transformType=TRANSFORMPOSSOURCE_View}) {b=true})
  n206 = Multiply(A=@n205, B=ScalarParameter() {defaultValue=10, group=SUPERLEVEL, parameterName=ChestRays_Scale, sortPriority=101})
  n207 = VectorParameter() {group=RelayStick_Orb, parameterName="OrbRadialCoordScales (UVUV)", defaultValue=<rgba 0.1366 2 0.2 1>}
  n208 = Multiply(
   A=MaterialFunctionCall(
    UV Coord=MaterialFunctionCall(UVs=@n203, Texture Scale=@n206) {materialFunction=&ScaleUVsByCenter}.UVs
    RadialScale=ScalarParameter() {defaultValue=1, parameterName=OutwardLightning_RadialScale}
   ) {materialFunction=&MF_PolarCoordinates}.Result
   B=AppendVector(A=@n207.B, B=@n207.B)
  )
  n209 = Saturate(None=Multiply(A=ComponentMask(None=@n208) {r=true}) {constB=10})
  n210 = Time()
  n211 = NamedRerouteUsage()
  n212 = TextureSample!(
   UVs=Add(A=Panner(Coordinate=@n208, Time=@n210) {speedX=-0.785, speedY=-0.023}, B=Multiply(A=@n211) {constB=0.025})
   Tex
   Level=Constant() {r=-1}
   Apply View MipBias
  ) {
   mipValueMode=TMVM_MipLevel
   samplerSource=SSM_Wrap_WorldGroupSettings
   samplerType=SAMPLERTYPE_LinearColor
   texture=/BRCosmetics/Characters/Player/Female/Medium/Bodies/F_MED_SpatialTravel_Watt/Skins/Superlevels/Textures/T_ElectricNoise
  }
  n213 = ScalarParameter() {defaultValue=2, group=SUPERLEVEL, parameterName=LightningPower}
  n214 = Multiply(A=Multiply(A=Multiply(A=@n209, B=@n209), B=Power(Base=@n212.R, Exp=@n213)))
  n215 = NamedRerouteUsage()
  n216 = Multiply(A=@n205, B=ScalarParameter() {defaultValue=5, group=SUPERLEVEL, parameterName=ChestRays_Scale02, sortPriority=101})
  n217 = VectorParameter() {group=RelayStick_Orb, parameterName="OrbRadialCoordScales (UVUV)", defaultValue=<rgba 0.1366 2 0.2 1>}
  n218 = Multiply(
   A=MaterialFunctionCall(
    UV Coord=MaterialFunctionCall(UVs=@n215, Texture Scale=@n216) {materialFunction=&ScaleUVsByCenter}.UVs
    RadialScale=ScalarParameter() {defaultValue=1, parameterName=OutwardLightning_RadialScale}
   ) {materialFunction=&MF_PolarCoordinates}.Result
   B=AppendVector(A=@n217.B, B=@n217.B)
  )
  n219 = Saturate(None=Multiply(A=ComponentMask(None=@n218) {r=true}) {constB=10})
  n220 = Time()
  n221 = VectorParameter() {group=RelayStick_Orb, parameterName="OrbRadialCoordPanSpeeds (UVUV)", defaultValue=<rgba 0.0333 0.0333 -0.6 -0.0233>}
  n222 = AppendVector(A=@n221.B, B=@n221.B)
  n223 = NamedRerouteUsage()
  n224 = TextureSample(
   UVs=Add(A=Panner(Coordinate=@n218, Time=@n220, Speed=@n222) {speedX=-0.33, speedY=-0.053}, B=Multiply(A=@n223) {constB=0.025})
   Level=Constant() {r=-1}
  ) {
   mipValueMode=TMVM_MipLevel
   samplerSource=SSM_Wrap_WorldGroupSettings
   samplerType=SAMPLERTYPE_LinearColor
   texture=&T_ElectricNoise
  }
  n225 = ScalarParameter() {defaultValue=2, group=SUPERLEVEL, parameterName=LightningPower}
  n226 = Multiply(A=Multiply(A=Multiply(A=@n219, B=@n219), B=Power(Base=@n224.R, Exp=@n225)))
  NamedRerouteDeclaration(None=Add(A=Add(A=@n214, B=@n226)) {constB=0.05}) {name=ChestRays, nodeColor=<rgba 0.1719 1 0 1>}
 ]
}
