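# Parent material for Tier 2 Academia superlevel character skins, turning outfits gold and metallic with animated, glowing swirls.
# PreFX and InitialSetup functions supply base attributes, which mask textures blend into metal, dark, swirl and secondary-color layers, with emissive outlines, before PostFX.
# Time-panned noise drives swirl distortion and wispy gradients, alongside body-position masks like ankle glow.
# Static switches add character-specific handling for PolarityWinn, LazyLizz and AshenMagnus, including chip decal, chest and banner.
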Material /BRCosmetics/Characters/Player/Female/Medium/Bodies/F_MED_SpatialTravel_Watt/Skins/Superlevels/ParentMaterials/M_AcademiaSL_T2_PARENT
graph {
 properties {
  bUseMaterialAttributes=true
  bUsedWithMeshDeformer=true
  bUsedWithMorphTargets=true
  bUsedWithSkeletalMesh=true
  pixelDepthOffsetMode=PDOM_Legacy
  shadingModel=MSM_FromMaterialExpression
  usedShadingModels=Default Lit | Subsurface Profile
 }
 nodes [
  n1 = NamedRerouteUsage()
  Add(A=@n1) {constB=0.5}
  Constant()
  n2 = MaterialFunctionCall!(Overide Texture Set UVs, Override Tech Art Mask UVs, Override Basic Emissive UVs) {materialFunction=/BRCosmetics/Characters/Player/Female/Medium/Bodies/F_MED_SpatialTravel_Watt/Skins/Superlevels/ParentMaterials/PreFX/MF_AcademiaSL_PreFX}
  MaterialFunctionCall(Material=NamedRerouteUsage(), M=@n2.M, Pre Skinned Local Position=@n2.M, Distance Blend=@n2.M, Fresnel=@n2.M) {materialFunction=&MF_CharacterParent_2023_PostFX}
  n3 = TextureSample() {
   samplerSource=SSM_Wrap_WorldGroupSettings
   samplerType=SAMPLERTYPE_Masks
   texture=/BRCosmetics/Characters/Player/Male/Medium/Bodies/M_MED_AshenMagnus_Slant/Textures/T_AshenMagnus_Slant_Body_M
  }
  n4 = NamedRerouteUsage()
  n5 = VectorParameter() {group=Superlevel, parameterName=Metal_Color, sortPriority=30, defaultValue=<rgba 0.8142 0.5328 0.0914 1>}
  n6 = BlendMaterialAttributes(
   A=@n4
   B=SetMaterialAttributes!(MaterialAttributes=@n4, Base Color=@n5.RGB, Metallic=Constant() {r=1}) {attributeSetTypes [#1, #2]}
   Alpha=ComponentMask(None=NamedRerouteUsage()) {g=true}
  )
  n7 = TextureSampleParameter2D() {
   group=Superlevel
   parameterName=SL_Masks_01
   samplerSource=SSM_Wrap_WorldGroupSettings
   samplerType=SAMPLERTYPE_Masks
   sortPriority=0
   texture=/BRCosmetics/Characters/Player/Female/Medium/Bodies/F_MED_SpatialTravel_Quirk/Textures/T_SpatialTravel_Quirk_Body_SL_FX
  }
  n8 = Saturate(None=Add(A=Add(A=@n7.R, B=@n7.R), B=@n7.B))
  n9 = NamedRerouteUsage()
  n10 = BlendMaterialAttributes(
   A=@n6
   B=SetMaterialAttributes!(MaterialAttributes=@n6, Base Color=Constant() {r=0.05}) {attributeSetTypes [#1]}
   Alpha=Subtract(A=OneMinus(None=@n8), B=ComponentMask(None=@n9) {b=true})
  )
  n11 = Constant3Vector() {constant=<rgba 1 0.786 0.2389 1>}
  n12 = Multiply(
   A=Power(Base=MaterialFunctionCall() {materialFunction=&BoundingBoxBased_0-1_UVW}.B) {constExponent=5}
   B=@n11
  )
  n13 = Constant3Vector() {constant=<rgba 1 0.4654 0 1>}
  n14 = NamedRerouteUsage()
  n15 = NamedRerouteUsage()
  n16 = NamedRerouteUsage()
  n17 = Multiply(A=Multiply(A=Power(Base=@n15)) {constB=1.5}, B=Saturate(None=Add(A=@n16) {constB=0.2}))
  n18 = LinearInterpolate(A=Max(A=Multiply(A=@n12) {constB=0.07}) {constB=0.01}, B=@n13, Alpha=Add(A=@n14, B=@n17))
  n19 = Saturate(None=@n1)
  n20 = Add(A=Multiply(A=Multiply(A=@n17, B=@n11), B=NamedRerouteUsage()), B=NamedRerouteUsage())
  n21 = NamedRerouteUsage()
  n22 = NamedRerouteUsage()
  n23 = BlendMaterialAttributes(
   A=@n10
   B=SetMaterialAttributes!(MaterialAttributes=@n10, Base Color=Multiply(A=@n18, B=@n19), Emissive Color=Add(A=@n20, B=@n21), Metallic=Multiply(A=@n14, B=@n22)) {attributeSetTypes [#1, #3, #2]}
   Alpha=ComponentMask(None=NamedRerouteUsage()) {r=true}
  )
  n24 = Constant()
  n25 = VectorParameter() {group=Superlevel, parameterName=Secondary_Color, sortPriority=30, defaultValue=<rgba 0.0434 0.0412 0.0229 1>}
  n26 = Constant() {r=0.5}
  n27 = NamedRerouteUsage()
  n28 = GetMaterialAttributes(None=BlendMaterialAttributes(
    A=@n23
    B=SetMaterialAttributes!(MaterialAttributes=@n23, Emissive Color=@n24, Base Color=@n25.RGB, Metallic=@n26) {attributeSetTypes [#3, #1, #2]}
    Alpha=ComponentMask(None=@n27) {b=true}
   )) {attributeGetTypes [#3]}
  n29 = VectorParameter() {group=Superlevel, parameterName=EmissiveDefaultVariant_Color, sortPriority=30, defaultValue=<rgba 2 0.7555 0.0162 1>}
  n30 = Multiply(A=@n29.RGB, B=OneMinus(None=NamedRerouteUsage()))
  n31 = Multiply(A=@n30) {constB=0.7}
  n32 = VectorParameter() {group=Superlevel, parameterName=Outlines_Color, sortPriority=30, defaultValue=<rgba 1 0.6699 0.114 1>}
  n33 = Desaturation(None=Multiply(A=@n32.RGB) {constB=1.5}, Fraction=Constant() {r=-1})
  n34 = TextureSampleParameter2D() {
   group=Superlevel
   parameterName=SL_Masks_02
   samplerSource=SSM_Wrap_WorldGroupSettings
   samplerType=SAMPLERTYPE_Masks
   sortPriority=0
   texture=/BRCosmetics/Characters/Player/Female/Medium/Bodies/F_MED_SpatialTravel_Quirk/Textures/T_SpatialTravel_Quirk_Body_SL_FX2
  }
  n35 = TextureCoordinate()
  n36 = Constant() {r=0.25}
  n37 = Constant() {r=0.05}
  n38 = Constant() {r=0.4}
  n39 = Constant() {r=0.59}
  n40 = SphereMask(A=NamedRerouteUsage(), B=Constant3Vector() {constant=<rgba 0.6584 0.6584 0.6584 1>}) {attenuationRadius=0.4, hardnessPercent=1}
  n41 = StaticSwitchParameter(
   True=Multiply(
    A=MaterialFunctionCall!(Texture Coordinate=@n35, X Pos Right=@n36, X Pos Left=@n37, Y Pos Up=@n38, Y Pos Down=@n39, SquareSoftness, RotateSquare, RotationAmount) {materialFunction=/Game/Accessories/FORT_Backpacks/Backpack_F_MED_GroovyReader_Play/Materials/MaterialFunctions/MF_SquareMaskGenerator}.SquareMask
    B=Multiply(A=@n40) {constB=1.5}
   )
   False=Constant()
  ) {group=Superlevel, parameterName=UseLazyLizzChipDecal}
  n42 = Add(A=@n34.G, B=@n41)
  n43 = LinearInterpolate(
   A=LinearInterpolate(
    A=@n28.Emissive Color
    B=StaticSwitchParameter(True=@n31, False=@n33) {group=Superlevel, parameterName=UsePolarityWinnFaceAcc}
    Alpha=Saturate(None=@n42)
   )
   B=@n30
   Alpha=Saturate(None=Add(A=@n34.R, B=@n41))
  )
  n44 = NamedRerouteUsage()
  n45 = Constant() {r=1}
  n46 = StaticSwitchParameter(
   True=MaterialFunctionCall(Hue Shift Percentage=ScalarParameter() {parameterName=AshenMagnusChest_HueShift}, Texture=@n44) {materialFunction=&HueShift}.Result
   False=Multiply(
    A=Desaturation(None=@n44, Fraction=@n45)
    B=VectorParameter() {group=Superlevel, parameterName=AshenMagnus_ChestTint, defaultValue=<rgba 1 1 1 0>}.RGB
   )
  ) {group=Superlevel, parameterName=AshenMagnus_UseHueShiftOnChest}
  n47 = NamedRerouteUsage()
  n48 = SetMaterialAttributes!(
   MaterialAttributes=@n28.MaterialAttributes
   Emissive Color=StaticSwitchParameter(True=LinearInterpolate(A=@n43, B=@n46, Alpha=@n47), False=@n43) {group=Superlevel, parameterName=UseAshenMagnusChest}
  ) {attributeSetTypes [#3]}
  NamedRerouteDeclaration(None=StaticSwitchParameter(
    True=MaterialFunctionCall!(Gmap=@n3.RGB, Materials=@n48) {materialFunction=/BRCosmetics/Characters/Player/Male/Medium/Bodies/M_MED_AshenMagnus_Slant/Materials/MF_AshenMagnus_PlayerBanner}.Result
    False=@n48
   ) {group=Superlevel, parameterName=UseAshenMagnusBanner}) {name=OutputMA, nodeColor=<rgba 0.2422 0 1 1>}
  n49 = TextureCoordinate()
  n50 = Time()
  n51 = Add(A=NamedRerouteUsage(), B=Multiply(A=Multiply(A=@n50) {constB=7}, B=Constant3Vector() {constant=<rgba -1 0 -1 0>}))
  n52 = TextureSampleParameter2D() {
   group=Superlevel
   parameterName=SL_Masks_03
   samplerSource=SSM_Wrap_WorldGroupSettings
   samplerType=SAMPLERTYPE_Masks
   texture=/BRCosmetics/Characters/Player/Female/Medium/Bodies/F_MED_PolarityWinn_Field/Textures/T_PolarityWinn_Field_Body_SL_FX3
  }
  n53 = Power(Base=@n52.G)
  n54 = Multiply(
   A=Multiply(A=Noise!(World Position=@n51, FilterWidth) {levelScale=1, levels=3, scale=0.018}) {constB=0.007}
   B=StaticSwitchParameter(True=Constant() {r=1}, False=OneMinus(None=@n53)) {group=Superlevel, parameterName=UseLazyLizzSkin}
  )
  n55 = NamedRerouteUsage()
  n56 = Multiply(A=Add(A=Multiply(A=@n55) {constB=2}) {constB=-1}, B=NamedRerouteUsage()) {constB=0.003}
  n57 = TextureSampleParameter2D(UVs=Add(A=Add(A=@n49, B=@n54) {constB=0}, B=Multiply(A=@n56) {constB=0.0052}) {constB=0}) {
   group=Superlevel
   parameterName=SL_Masks_03
   samplerSource=SSM_Wrap_WorldGroupSettings
   samplerType=SAMPLERTYPE_Masks
   texture=&T_PolarityWinn_Field_Body_SL_FX3
  }
  NamedRerouteDeclaration(None=Saturate(None=@n57.R)) {name=YellowSwirlsBase, nodeColor=<rgba 1 0 0.5625 1>}
  n58 = Saturate(None=Add(A=@n14) {constB=0.01})
  n59 = Multiply(A=NamedRerouteUsage(), B=@n13)
  NamedRerouteDeclaration(None=Multiply(A=Multiply(A=@n58, B=@n59)) {constB=3}) {name=GoldenSwirls, nodeColor=<rgba 0.3125 0 1 1>}
  n60 = MaterialFunctionCall!(
   AshenMagnus_OriginalEmissive=NamedRerouteUsage()
   Result=@n2.Result
   Tech Art Mask=@n2.Result
   M=NamedRerouteDeclaration(None=@n2.M) {name=InMask, nodeColor=<rgba 0 1 0.25 1>}
   Pre skinned Local Position=NamedRerouteDeclaration(None=@n2.Pre skinned Local Position) {name=PreskinnedLocalPosition, nodeColor=<rgba 0 0.0156 1 1>}
   Fresnel=NamedRerouteDeclaration(None=@n2.Fresnel) {name=InFresnel, nodeColor=<rgba 1 0 0.8203 1>}
  ) {materialFunction=&/MF_AcademiaSL_InitialSetup}
  NamedRerouteDeclaration(None=@n60.OriginalBaseColor) {name=OriginalBaseColor, nodeColor=<rgba 1 0 0.3516 1>}
  n61 = TextureSample() {
   samplerSource=SSM_Wrap_WorldGroupSettings
   texture=/BRCosmetics/Characters/Player/Male/Medium/Bodies/M_MED_AshenMagnus_Slant/Skins/Rye/Textures/T_AshenMagnus_SlantRye_Body_E
  }
  n62 = NamedRerouteDeclaration(None=@n61.RGB) {name=AshenMagnus_OriginalEmissive, nodeColor=<rgba 0 0.9297 1 1>}
  n63 = MaterialFunctionCall(Float3=Saturate(None=Multiply(A=@n62) {constB=2})) {materialFunction=&BreakOutFloat3Components}
  n64 = Multiply(
   A=Max(A=Max(A=@n63.R, B=@n63.R), B=@n63.B)
   B=MaterialFunctionCall!(A=TextureCoordinate(), B=Constant2Vector() {g=0.215, r=0.365}, Bounds=Constant2Vector() {g=0.1, r=0.1}, Edge Falloff=Constant()) {materialFunction=/Engine/Functions/Engine_MaterialFunctions01/Gradient/BoxMask-2D}.Result
  )
  NamedRerouteDeclaration(None=Saturate(None=Multiply(A=@n64) {constB=2})) {name=AshenMagnus_ChestMask, nodeColor=<rgba 1 0 0.5391 1>}
  NamedRerouteDeclaration(None=@n60.OriginalEmissive) {name=OriginalEmissive, nodeColor=<rgba 0.1484 0 1 1>}
  n65 = Add(A=NamedRerouteUsage()) {constB=0.5}
  n66 = ComponentMask(None=NamedRerouteUsage()) {b=true}
  n67 = Saturate(None=Divide(A=@n66) {constB=40})
  n68 = Saturate(None=Divide(A=@n66) {constB=3})
  n69 = Divide(A=@n66) {constB=10}
  n70 = Multiply(A=OneMinus(None=Saturate(None=@n69))) {constB=20}
  n71 = Multiply(
   A=Constant3Vector() {constant=<rgba 0.8108 0.4411 0.0009 1>}
   B=LinearInterpolate(A=OneMinus(None=@n67), B=OneMinus(None=@n68), Alpha=Saturate(None=@n70))
  )
  n72 = Saturate(None=Add(A=NamedRerouteUsage()) {constB=0.3})
  NamedRerouteDeclaration(None=Multiply(A=Saturate(None=@n65), B=Multiply(A=@n71, B=@n72))) {name=AnkleGlow, nodeColor=<rgba 1 0.375 0 1>}
  n73 = Time()
  n74 = Multiply(A=Sine(None=Multiply(A=@n73) {constB=0.19})) {constB=0.5}
  n75 = NamedRerouteUsage()
  n76 = Rotator(Coordinate=ComponentMask(None=Divide(A=@n75) {constB=80}) {b=true, r=true}, Time=Constant() {r=3.1415})
  n77 = Multiply(A=@n76, B=ScalarParameter() {defaultValue=3, parameterName=WispyFire_Scale})
  n78 = Multiply(A=Time(), B=Constant2Vector() {g=0.2, r=0.01}) {constB=0.2}
  n79 = Add(A=Multiply(A=NamedRerouteUsage()) {constB=2}) {constB=-1}
  n80 = TextureSample(UVs=Add(A=Subtract(A=@n77, B=@n78), B=Multiply(A=@n79) {constB=0.2}) {constB=0}) {
   samplerSource=SSM_Wrap_WorldGroupSettings
   texture=/BRCosmetics/Characters/Player/Female/Medium/Bodies/F_MED_SpatialTravel_Watt/Skins/Superlevels/Textures/T_WispyFireNoise
  }
  n81 = NamedRerouteDeclaration(None=@n80.R) {name=WispyFireNoise, nodeColor=<rgba 0 0.8125 1 1>}
  n82 = Subtract(A=ComponentMask(None=@n76) {g=true}, B=Multiply(A=Time()) {constB=0.39})
  n83 = Add(A=Multiply(A=Sine(None=@n82)) {constB=0.5}) {constB=0.5}
  n84 = Saturate(None=Add(A=Power(Base=@n83) {constExponent=3}) {constB=0.1})
  n85 = Add(
   A=Multiply(A=Add(A=@n74) {constB=0.5}) {constB=0.3}
   B=Multiply(A=Add(A=@n81) {constB=0.1}, B=NamedRerouteDeclaration(None=@n84) {name=BodyGradient, nodeColor=<rgba 0 0.0859 1 1>})
  ) {constA=0.5}
  NamedRerouteDeclaration(None=Multiply(A=Power(Base=@n85)) {constB=3}) {name=PanningGradient, nodeColor=<rgba 0.0312 0 1 1>}
  n86 = Multiply(A=Time()) {constB=10}
  n87 = Constant3Vector() {constant=<rgba -0.31 0 -1 0>}
  n88 = PreSkinnedNormal()
  n89 = MaterialFunctionCall!(
   Export Float 4
   TextureObject=TextureObject() {texture=/BRCosmetics/Characters/Player/Female/Medium/Bodies/F_MED_SpatialTravel_Watt/Skins/Superlevels/Textures/T_AcademiaSL_T1_Masks}
   TextureSize=Multiply(A=ScalarParameter() {defaultValue=1.1, parameterName=DistortionNoise_Size}) {constB=64}
   WorldPosition=Add(A=NamedRerouteUsage(), B=Multiply(A=@n86, B=@n87))
   World Space Normal=ShaderStageSwitch(PixelShader=VertexInterpolator(VS=@n88).PS, VertexShader=@n88)
   RotZAxis
   RotXAxis
   RotYAxis
   ProjectionTransitionContrast
  ) {materialFunction=&/MF_WorldAlignedTexture_SL}
  n90 = ComponentMask(None=@n89.XYZ Texture) {b=true}
  NamedRerouteDeclaration(None=Saturate(None=Add(A=@n90) {constB=0.05})) {name=NoiseDistortion, nodeColor=<rgba 1 0.0938 0 1>}
  NamedRerouteDeclaration(None=@n60.Result) {name=OriginalMA, nodeColor=<rgba 0.1719 0 1 1>}
  NamedRerouteDeclaration(None=@n7.RGB) {name=SL_Masks, nodeColor=<rgba 0.1484 1 0 1>}
  n91 = NamedRerouteUsage()
  n92 = VectorParameter() {parameterName=CrotchMask_Position, defaultValue=<rgba 0 0 80 0>}
  n93 = ScalarParameter() {defaultValue=40, parameterName=CrotchMask_Radius}
  NamedRerouteDeclaration(None=OneMinus(None=SphereMask(A=@n91, B=@n92.RGB, Radius=@n93) {attenuationRadius=0, hardnessPercent=0.2})) {name=CrotchMask, nodeColor=<rgba 1 0 0.0469 1>}
 ]
}
