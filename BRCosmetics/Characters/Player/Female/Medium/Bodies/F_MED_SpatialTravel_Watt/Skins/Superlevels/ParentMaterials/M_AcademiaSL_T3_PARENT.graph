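# Parent character material for Academia tier 3 Superlevel skins, layering animated energy effects over the original material.
# A camera-facing virtual plane at a Blueprint-driven or overridden chest position supplies UVs for the pulsing core, panning polar rays and lightning.
# Triplanar animated cubes and pre-skinned sphere masks add body-localized detail.
# Mask textures blend metal, dark, effect and secondary-color layers plus emissive outlines, with static switches for skin-specific banner, chest and decal variants.
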Material /BRCosmetics/Characters/Player/Female/Medium/Bodies/F_MED_SpatialTravel_Watt/Skins/Superlevels/ParentMaterials/M_AcademiaSL_T3_PARENT
graph {
 properties {
  bUseMaterialAttributes=true
  bUsedWithMeshDeformer=true
  bUsedWithMorphTargets=true
  bUsedWithSkeletalMesh=true
  pixelDepthOffsetMode=PDOM_Legacy
  shadingModel=MSM_FromMaterialExpression
  usedShadingModels=Default Lit | Subsurface Profile
 }
 nodes [
  n1 = Time()
  n2 = NamedRerouteUsage()
  n3 = Multiply(A=NamedRerouteUsage(), B=ScalarParameter() {defaultValue=30, group=SUPERLEVEL, parameterName=ChestPattern_Scale, sortPriority=101})
  n4 = Time()
  n5 = TextureSample(UVs=Rotator(
    Coordinate=MaterialFunctionCall(UVs=@n2, Texture Scale=@n3) {materialFunction=&ScaleUVsByCenter}.UVs
    Time=Multiply(A=@n4) {constB=0.275}
   )) {
   samplerSource=SSM_Clamp_WorldGroupSettings
   texture=/BRCosmetics/Characters/Player/Female/Medium/Bodies/F_MED_SpatialTravel_Watt/Skins/Superlevels/Textures/T_AcademiaSL_T3_Masks
  }
  n6 = NamedRerouteDeclaration(None=@n5.R) {name=ChestAreaMask, nodeColor=<rgba 0.8751 1 0.8076 1>}
  n7 = Multiply(A=@n6) {constB=3}
  Add(A=@n1, B=@n7)
  n8 = MaterialFunctionCall!(Overide Texture Set UVs, Override Tech Art Mask UVs, Override Basic Emissive UVs) {materialFunction=/BRCosmetics/Characters/Player/Female/Medium/Bodies/F_MED_SpatialTravel_Watt/Skins/Superlevels/ParentMaterials/PreFX/MF_AcademiaSL_PreFX}
  n9 = NamedRerouteDeclaration(None=@n8.Pre skinned Local Position) {name=PreskinnedLocalPosition, nodeColor=<rgba 0 0.2969 1 1>}
  MaterialFunctionCall(Material=NamedRerouteUsage(), M=@n8.M, Pre Skinned Local Position=@n9, Distance Blend=@n8.M, Fresnel=@n8.M) {materialFunction=&MF_CharacterParent_2023_PostFX}
  n10 = Time()
  n11 = PreSkinnedNormal()
  n12 = ShaderStageSwitch(PixelShader=VertexInterpolator(VS=@n11).PS, VertexShader=@n11)
  n13 = TextureObject() {texture=/BRCosmetics/Characters/Player/Female/Medium/Bodies/F_MED_SpatialTravel_Watt/Skins/Superlevels/Textures/T_AcademiaSL_T3_Masks_02}
  n14 = NamedRerouteUsage()
  n15 = MaterialFunctionCall!(VertexNormal=@n12, Basecolor Texture=@n13, Worldpos=@n14, TextureSize=Multiply(A=Constant() {r=0.4}) {constB=64}) {materialFunction=&/SimpleLowEndTriplanar_SL}
  n16 = Sine(None=Add(A=Multiply(A=@n10) {constB=0.39}, B=ComponentMask(None=@n15.Result) {r=true}))
  n17 = Max(A=Add(A=Multiply(A=@n16) {constB=0.5}) {constB=0.5}) {constB=0.1}
  n18 = ComponentMask(None=@n15.Result) {g=true}
  n19 = Power(Base=NamedRerouteUsage()) {constExponent=3}
  n20 = NamedRerouteUsage()
  n21 = Divide(A=@n20) {constB=50}
  n22 = OneMinus(None=ComponentMask(None=Saturate(None=@n21)) {b=true})
  n23 = Add(A=Saturate(None=Multiply(A=@n19) {constB=3}), B=Multiply(A=Multiply(A=@n22, B=@n22)) {constB=2})
  n24 = Multiply(A=Multiply(A=Multiply(A=@n17, B=@n18), B=Saturate(None=@n23))) {constB=2}
  n25 = MaterialFunctionCall(VertexNormal=@n12, Basecolor Texture=@n13, Worldpos=@n14, TextureSize=Multiply(A=Constant() {r=0.4}) {constB=128}) {materialFunction=&SimpleLowEndTriplanar_SL}
  n26 = Add(A=Multiply(A=Time()) {constB=0.39}, B=ComponentMask(None=@n25.Result) {r=true})
  n27 = Add(A=Multiply(A=Sine(None=@n26)) {constB=0.5}) {constB=0.5}
  n28 = Multiply(A=Multiply(A=Max(A=@n27) {constB=0.1}, B=ComponentMask(None=@n25.Result) {g=true})) {constB=0.07}
  n29 = Divide(A=@n20) {constB=4}
  n30 = OneMinus(None=ComponentMask(None=Saturate(None=@n29)) {b=true})
  n31 = Saturate(None=Multiply(A=Multiply(A=@n30, B=@n30)) {constB=2})
  NamedRerouteDeclaration(None=LinearInterpolate(A=Add(A=@n24, B=@n28), Alpha=@n31) {constB=0.2}) {name=Cubes, nodeColor=<rgba 0.2656 1 0 1>}
  n32 = NamedRerouteUsage()
  n33 = Constant3Vector() {constant=<rgba 50 0 -75 0>}
  n34 = ComponentMask(None=Divide(A=Add(A=@n32, B=@n33), B=ScalarParameter() {defaultValue=100, parameterName=AAS}) {constB=170}) {b=true, r=true}
  n35 = ScalarParameter() {defaultValue=3, group=Superlevel, parameterName=AdditionalRays_LenghtScale, sortPriority=70}
  n36 = ScalarParameter() {defaultValue=0.3, group=Superlevel, parameterName=AdditionalRays_RadialScale, sortPriority=70}
  n37 = TextureSample(UVs=Panner(Coordinate=MaterialFunctionCall!(UV Coord=@n34, LenghtScale=@n35, RadialScale=@n36) {materialFunction=/Game/Global/MaterialFunctions/TechArt/MF_PolarCoordinates}.Result) {speedX=1.72, speedY=-0.025}) {
   samplerSource=SSM_Wrap_WorldGroupSettings
   texture=/BRCosmetics/Characters/Player/Female/Medium/Bodies/F_MED_SpatialTravel_Watt/Skins/Superlevels/Textures/T_AcademiaSL_T3_Masks_03
  }
  n38 = PreSkinnedNormal()
  n39 = ShaderStageSwitch(PixelShader=VertexInterpolator(VS=@n38).PS, VertexShader=@n38)
  n40 = Multiply(A=NamedRerouteUsage()) {constB=20}
  n41 = Power(Base=Abs(None=ComponentMask(None=@n39) {r=true}), Exp=LinearInterpolate(Alpha=Saturate(None=@n40)) {constA=8}) {constExponent=1}
  n42 = OneMinus(None=Saturate(None=Multiply(A=@n41) {constB=5}))
  n43 = Multiply(A=Saturate(None=Multiply(A=@n37.G, B=@n42))) {constB=40}
  n44 = NamedRerouteUsage()
  n45 = ComponentMask(None=TransformPosition(None=NamedRerouteUsage()) {transformSourceType=TRANSFORMPOSSOURCE_World, transformType=TRANSFORMPOSSOURCE_View}) {b=true}
  n46 = MaterialFunctionCall(UVs=NamedRerouteUsage(), Texture Scale=Multiply(A=Divide(B=@n45), B=Constant() {r=60}) {constB=60}) {materialFunction=&ScaleUVsByCenter}
  n47 = Constant() {r=2}
  n48 = TextureSample(UVs=Panner(Coordinate=MaterialFunctionCall(UV Coord=@n46.UVs, LenghtScale=@n47, RadialScale=@n36) {materialFunction=&MF_PolarCoordinates}.Result) {speedX=0.735, speedY=0.04}) {
   samplerSource=SSM_Wrap_WorldGroupSettings
   texture=&T_AcademiaSL_T3_Masks_03
  }
  NamedRerouteDeclaration(None=Add(A=Multiply(A=@n43, B=@n44), B=Multiply(A=@n48.B) {constB=4})) {name=InwardRays_Additional, nodeColor=<rgba 1 0.4453 0 1>}
  n49 = NamedRerouteUsage()
  n50 = Constant3Vector() {constant=<rgba 50 5 88 0>}
  n51 = Constant() {r=35}
  n52 = Multiply(A=@n50, B=Constant3Vector() {constant=<rgba -1 1 1 0>})
  n53 = NamedRerouteDeclaration(None=Add(
    A=SphereMask(A=@n49, B=@n50, Radius=@n51) {attenuationRadius=30, hardnessPercent=0.3}
    B=SphereMask(A=@n49, B=@n52, Radius=@n51) {attenuationRadius=30, hardnessPercent=0.3}
   )) {name=HandsMask, nodeColor=<rgba 0 0.0859 1 1>}
  n54 = Constant3Vector() {constant=<rgba 0 0 100 0>}
  n55 = Multiply(
   A=Power(Base=SphereMask(A=@n49, B=@n54) {attenuationRadius=55, hardnessPercent=0}) {constExponent=3}
   B=ScalarParameter() {defaultValue=2.5, parameterName=TorsoBrightness}
  ) {constB=3}
  NamedRerouteDeclaration(None=Multiply(A=Add(A=@n53, B=@n55)) {constB=0.5}) {name=TorsoMask, nodeColor=<rgba 1 0.7969 0 1>}
  n56 = MaterialFunctionCall!(
   AshenMagnus_OriginalEmissive=NamedRerouteUsage()
   Result=@n8.Result
   Tech Art Mask=@n8.Result
   M=NamedRerouteDeclaration(None=@n8.M) {name=InMask, nodeColor=<rgba 1 0.9609 0 1>}
   Pre skinned Local Position=@n9
   Fresnel=NamedRerouteDeclaration(None=@n8.Fresnel) {name=InFresnel, nodeColor=<rgba 0.7109 0 1 1>}
  ) {materialFunction=&/MF_AcademiaSL_InitialSetup}
  NamedRerouteDeclaration(None=@n56.Result) {name=OriginalMA, nodeColor=<rgba 1 0.2344 0 1>}
  n57 = NamedRerouteUsage()
  n58 = ScalarParameter() {defaultValue=3, group=Superlevel, parameterName=CubesBrightness}
  n59 = Multiply(A=Power(Base=@n6) {constExponent=5}) {constB=0.5}
  n60 = Subtract(A=Multiply(A=@n4) {constB=0.8}, B=@n5.A)
  n61 = Add(A=Multiply(A=Sine(None=@n60)) {constB=0.5}) {constB=0.5}
  n62 = Multiply(A=Multiply(A=@n5.G, B=Add(A=@n61) {constB=0.1})) {constB=3}
  n63 = NamedRerouteUsage()
  n64 = Divide(B=ComponentMask(None=TransformPosition(None=@n63) {transformSourceType=TRANSFORMPOSSOURCE_World, transformType=TRANSFORMPOSSOURCE_View}) {b=true})
  n65 = MaterialFunctionCall(
   UVs=@n2
   Texture Scale=Multiply(A=NamedRerouteDeclaration(None=@n64) {name=ChestPosition_ViewSpaceScale, nodeColor=<rgba 1 0.422 0 1>}) {constB=60}
  ) {materialFunction=&ScaleUVsByCenter}
  n66 = TextureSample(UVs=Panner(Coordinate=MaterialFunctionCall(UV Coord=@n65.UVs) {materialFunction=&MF_PolarCoordinates}.Result) {speedX=0.2, speedY=0.01}) {
   samplerSource=SSM_Wrap_WorldGroupSettings
   texture=/BRCosmetics/Characters/Player/Female/Medium/Bodies/F_MED_SpatialTravel_Watt/Skins/Superlevels/Textures/T_AcademiaSL_T1_Masks
  }
  n67 = Add(A=Multiply(A=@n66.B) {constB=2}) {constB=-1}
  n68 = NamedRerouteUsage()
  n69 = Sine(None=Multiply(A=Time()) {constB=8})
  n70 = Multiply(
   A=NamedRerouteUsage()
   B=Add(A=ScalarParameter() {defaultValue=13, group=Superlevel, parameterName=BrightCore_Scale}, B=Multiply(A=@n69) {constB=0.5}) {constA=13}
  ) {constB=10}
  n71 = TextureSample(UVs=Add(
    A=Multiply(A=@n67) {constB=0.05}
    B=MaterialFunctionCall(UVs=@n68, Texture Scale=@n70) {materialFunction=&ScaleUVsByCenter}.UVs
   )) {
   samplerSource=SSM_Clamp_WorldGroupSettings
   texture=&T_AcademiaSL_T3_Masks
  }
  n72 = Add(A=Add(A=NamedRerouteUsage(), B=Add(A=@n59, B=@n62)), B=Multiply(A=@n71.B) {constB=2})
  n73 = NamedRerouteUsage()
  n74 = Multiply(A=Add(A=Add(A=@n72, B=@n73), B=NamedRerouteUsage()), B=ScalarParameter() {defaultValue=1.35, parameterName=RaysBrightness})
  n75 = SphereMask(A=WorldPosition().XYZ, B=NamedRerouteUsage()) {attenuationRadius=100, hardnessPercent=0.5}
  NamedRerouteDeclaration(None=Add(A=Multiply(A=@n57, B=@n58) {constB=2}, B=Multiply(A=@n74, B=@n75))) {name="Core&&Cubes", nodeColor=<rgba 0 0.0859 1 1>}
  n76 = NamedRerouteUsage()
  n77 = NamedRerouteUsage()
  n78 = VectorParameter() {group=Superlevel, parameterName=Metal_Color, sortPriority=30, defaultValue=<rgba 0.599 0.599 0.599 1>}
  n79 = BlendMaterialAttributes(
   A=@n77
   B=SetMaterialAttributes!(MaterialAttributes=@n77, Base Color=@n78.RGB, Metallic=Constant() {r=1}) {attributeSetTypes [#1, #2]}
   Alpha=ComponentMask(None=NamedRerouteUsage()) {g=true}
  )
  n80 = TextureSampleParameter2D() {
   group=Superlevel
   parameterName=SL_Masks_01
   samplerSource=SSM_Wrap_WorldGroupSettings
   samplerType=SAMPLERTYPE_Masks
   sortPriority=0
   texture=/BRCosmetics/Characters/Player/Female/Medium/Bodies/F_MED_SpatialTravel_Quirk/Textures/T_SpatialTravel_Quirk_Body_SL_FX
  }
  n81 = Saturate(None=Add(A=Add(A=@n80.R, B=@n80.R), B=@n80.B))
  n82 = NamedRerouteUsage()
  n83 = BlendMaterialAttributes(
   A=@n79
   B=SetMaterialAttributes!(MaterialAttributes=@n79, Base Color=Constant() {r=0.05}, Roughness=Constant() {r=0.325}, Emissive Color) {attributeSetTypes [#1, #3, #4]}
   Alpha=Subtract(A=OneMinus(None=@n81), B=ComponentMask(None=@n82) {b=true})
  )
  n84 = MaterialFunctionCall() {materialFunction=&BoundingBoxBased_0-1_UVW}
  n85 = Constant3Vector() {constant=<rgba 0.0885 0.3767 0.2491 1>}
  n86 = Multiply(A=Multiply(A=Power(Base=@n84.B) {constExponent=5}, B=@n85)) {constB=0.3}
  n87 = NamedRerouteUsage()
  n88 = Power(Base=NamedRerouteUsage()) {constExponent=3}
  n89 = Multiply(A=@n85, B=Add(A=NamedRerouteUsage(), B=Add(A=@n87, B=@n88)))
  n90 = NamedRerouteUsage()
  n91 = BlendMaterialAttributes(
   A=@n83
   B=SetMaterialAttributes!(MaterialAttributes=@n83, Base Color=Max(A=@n86) {constB=0.01}, Emissive Color=Multiply(A=@n89, B=@n90)) {attributeSetTypes [#1, #4]}
   Alpha=ComponentMask(None=NamedRerouteUsage()) {r=true}
  )
  n92 = Constant() {r=0.5}
  n93 = VectorParameter() {group=Superlevel, parameterName=Secondary_Color, sortPriority=30, defaultValue=<rgba 0.6784 0.8247 0.4656 1>}
  n94 = Constant() {r=1}
  n95 = NamedRerouteUsage()
  n96 = GetMaterialAttributes(None=BlendMaterialAttributes(
    A=@n91
    B=SetMaterialAttributes!(MaterialAttributes=@n91, Roughness=@n92, Base Color=@n93.RGB, Metallic=@n94) {attributeSetTypes [#3, #1, #2]}
    Alpha=ComponentMask(None=@n95) {b=true}
   )) {attributeGetTypes [#4]}
  n97 = VectorParameter() {group=Superlevel, parameterName=EmissiveDefaultVariant_Color, sortPriority=30, defaultValue=<rgba 0 2 1.2182 1>}
  n98 = Multiply(A=@n97.RGB, B=OneMinus(None=NamedRerouteUsage()))
  n99 = Multiply(A=@n98) {constB=0.7}
  n100 = VectorParameter() {group=Superlevel, parameterName=Outlines_Color, sortPriority=30, defaultValue=<rgba 0.2396 0.6545 0.4479 1>}
  n101 = Desaturation(None=Multiply(A=@n100.RGB) {constB=3.5}, Fraction=Constant() {r=-0.7})
  n102 = TextureSampleParameter2D() {
   group=Superlevel
   parameterName=SL_Masks_02
   samplerSource=SSM_Wrap_WorldGroupSettings
   samplerType=SAMPLERTYPE_Masks
   sortPriority=0
   texture=/BRCosmetics/Characters/Player/Female/Medium/Bodies/F_MED_SpatialTravel_Quirk/Textures/T_SpatialTravel_Quirk_Body_SL_FX2
  }
  n103 = TextureCoordinate()
  n104 = Constant() {r=0.25}
  n105 = Constant() {r=0.05}
  n106 = Constant() {r=0.4}
  n107 = Constant() {r=0.59}
  n108 = SphereMask(A=NamedRerouteUsage(), B=Constant3Vector() {constant=<rgba 0.6584 0.6584 0.6584 1>}) {attenuationRadius=0.4, hardnessPercent=1}
  n109 = StaticSwitchParameter(
   True=Multiply(
    A=MaterialFunctionCall!(
     Texture Coordinate=@n103
     X Pos Right=@n104
     X Pos Left=@n105
     Y Pos Up=@n106
     Y Pos Down=@n107
     SquareSoftness
     RotateSquare
     RotationAmount
    ) {materialFunction=/Game/Accessories/FORT_Backpacks/Backpack_F_MED_GroovyReader_Play/Materials/MaterialFunctions/MF_SquareMaskGenerator}.SquareMask
    B=Multiply(A=@n108) {constB=1.5}
   )
   False=Constant()
  ) {group=Superlevel, parameterName=UseLazyLizzChipDecal}
  n110 = Add(A=@n102.G, B=@n109)
  n111 = LinearInterpolate(
   A=LinearInterpolate(
    A=@n96.Emissive Color
    B=StaticSwitchParameter(True=@n99, False=@n101) {group=Superlevel, parameterName=UsePolarityWinnFaceAcc}
    Alpha=Saturate(None=@n110)
   )
   B=@n98
   Alpha=Saturate(None=Add(A=@n102.R, B=@n109))
  )
  n112 = NamedRerouteUsage()
  n113 = Constant() {r=1}
  n114 = StaticSwitchParameter(
   True=MaterialFunctionCall(Hue Shift Percentage=ScalarParameter() {parameterName=AshenMagnusChest_HueShift}, Texture=@n112) {materialFunction=&HueShift}.Result
   False=Multiply(
    A=Desaturation(None=@n112, Fraction=@n113)
    B=VectorParameter() {group=Superlevel, parameterName=AshenMagnus_ChestTint, defaultValue=<rgba 1 1 1 0>}.RGB
   )
  ) {group=Superlevel, parameterName=AshenMagnus_UseHueShiftOnChest}
  n115 = NamedRerouteUsage()
  n116 = SetMaterialAttributes!(
   MaterialAttributes=@n96.MaterialAttributes
   Emissive Color=StaticSwitchParameter(True=LinearInterpolate(A=@n111, B=@n114, Alpha=@n115), False=@n111) {group=Superlevel, parameterName=UseAshenMagnusChest}
  ) {attributeSetTypes [#4]}
  NamedRerouteDeclaration(None=StaticSwitchParameter(
    True=MaterialFunctionCall!(Gmap=@n76, Materials=@n116) {materialFunction=/BRCosmetics/Characters/Player/Male/Medium/Bodies/M_MED_AshenMagnus_Slant/Materials/MF_AshenMagnus_PlayerBanner}.Result
    False=@n116
   ) {group=Superlevel, parameterName=UseAshenMagnusBanner}) {name=OutputMA, nodeColor=<rgba 0.8047 1 0 1>}
  n117 = ScalarParameter() {defaultValue=7, group=Superlevel, parameterName=TorsoMask_RadiusMin}
  n118 = Multiply(A=ScalarParameter() {defaultValue=50, group=Superlevel, parameterName=TorsoMask_Radius}) {constB=0.5}
  n119 = Power(Base=SphereMask(A=@n49, B=@n54, Radius=Add(A=@n117, B=@n118) {constA=10}) {attenuationRadius=55, hardnessPercent=0}) {constExponent=3}
  NamedRerouteDeclaration(None=Multiply(A=Add(A=@n53, B=@n119)) {constB=0.5}) {name=TorsoMask_RandomLenght, nodeColor=<rgba 1 0.7969 0 1>}
  n120 = TextureSample() {
   samplerSource=SSM_Wrap_WorldGroupSettings
   texture=/BRCosmetics/Characters/Player/Male/Medium/Bodies/M_MED_AshenMagnus_Slant/Skins/Rye/Textures/T_AshenMagnus_SlantRye_Body_E
  }
  n121 = NamedRerouteDeclaration(None=@n120.RGB) {name=AshenMagnus_OriginalEmissive, nodeColor=<rgba 0 0.9297 1 1>}
  n122 = MaterialFunctionCall(Float3=Saturate(None=Multiply(A=@n121) {constB=2})) {materialFunction=&BreakOutFloat3Components}
  n123 = Multiply(
   A=Max(A=Max(A=@n122.R, B=@n122.R), B=@n122.B)
   B=MaterialFunctionCall!(A=TextureCoordinate(), B=Constant2Vector() {g=0.215, r=0.365}, Bounds=Constant2Vector() {g=0.1, r=0.1}, Edge Falloff=Constant()) {materialFunction=/Engine/Functions/Engine_MaterialFunctions01/Gradient/BoxMask-2D}.Result
  )
  NamedRerouteDeclaration(None=Saturate(None=Multiply(A=@n123) {constB=2})) {name=AshenMagnus_ChestMask, nodeColor=<rgba 1 0 0.5391 1>}
  NamedRerouteDeclaration(None=@n56.OriginalEmissive) {name=OriginalEmissive, nodeColor=<rgba 0.1484 0 1 1>}
  NamedRerouteDeclaration(None=@n56.OriginalBaseColor) {name=OriginalBaseColor, nodeColor=<rgba 1 0 0.375 1>}
  n124 = Subtract(A=@n1, B=@n7)
  n125 = Add(A=Multiply(A=Sine(None=@n124)) {constB=0.5}) {constB=0.5}
  n126 = Saturate(None=Multiply(A=@n6) {constB=2})
  NamedRerouteDeclaration(None=Multiply(A=Multiply(A=@n125, B=@n126)) {constB=0.05}) {name=PulsatingEffect, nodeColor=<rgba 0.2891 0 1 1>}
  n127 = MaterialFunctionCall() {materialFunction=&CameraDirectionVector}
  n128 = Multiply(A=Constant3Vector() {constant=<rgba 1 1 0 0>}, B=@n127.Result)
  n129 = Add(A=Multiply(B=Normalize(VectorInput=@n128)) {constA=0.25}, B=@n127.Result)
  n130 = VectorParameter() {group=Superlevel, parameterName=CorePosition_Override, sortPriority=600000, defaultValue=<rgba 0 0 125 0>}
  n131 = VectorParameter() {group=Superlevel, parameterName=LocalChestPosition, sortPriority=1000, defaultValue=<rgba 0 0 127 0>}
  n132 = TransformPosition(None=Add(A=@n131.RGB, B=VectorParameter() {group=Superlevel, parameterName=PreviewOffset_Chest, sortPriority=800}.RGB)) {transformType=TRANSFORMPOSSOURCE_World}
  n133 = VectorParameter() {group=Superlevel, parameterName=BP_OffsetVector_Chest, sortPriority=1000}
  n134 = Add(
   A=Multiply(
    A=ScalarParameter() {defaultValue=-3, group=SUPERLEVEL, parameterName=PositionOffset_Chest, sortPriority=150}
    B=Normalize(VectorInput=@n133.RGB)
   )
   B=VectorParameter() {group=Superlevel, parameterName=BP_ChestPosition, sortPriority=1000}.RGB
  )
  n135 = ScalarParameter() {group=SUPERLEVEL, parameterName=BP_UseCCPM, sortPriority=0}
  n136 = NamedRerouteDeclaration(None=StaticSwitchParameter(True=TransformPosition(None=@n130.RGB) {transformType=TRANSFORMPOSSOURCE_World}, False=LinearInterpolate(A=@n132, B=@n134, Alpha=@n135)) {group=SUPERLEVEL, parameterName=OverrideCorePosition, sortPriority=600000}) {name=ChestPosition, nodeColor=<rgba 0 1 0.2734 1>}
  n137 = Constant() {r=45}
  n138 = Constant() {r=90}
  n139 = MaterialFunctionCall() {materialFunction=/Engine/Functions/Engine_MaterialFunctions02/Utility/FOV}
  n140 = Multiply(
   A=LinearInterpolate(A=Constant() {r=64}, B=Constant() {r=128}, Alpha=InverseLinearInterpolate(A=@n137, B=@n138, Value=@n139.FOV angle in degrees))
   B=Distance(A=@n136, B=CameraPositionWS())
  )
  NamedRerouteDeclaration(None=MaterialFunctionCall(Plane Normal Axis=Normalize(VectorInput=@n129), Plane Center=Add(A=@n127.Result, B=@n136), UV Scale=Multiply(A=@n140) {constB=0.0146}) {materialFunction=&VirtualPlaneCoordinates}.UVs) {name=ChestAllignedScreenPosition, nodeColor=<rgba 0.125 1 0 1>}
  NamedRerouteDeclaration(None=@n80.RGB) {name=SL_Masks, nodeColor=<rgba 1 0 0.3516 1>}
  n141 = NamedRerouteUsage()
  n142 = NamedRerouteUsage()
  n143 = OneMinus(None=Saturate(None=Multiply(A=@n142) {constB=2}))
  n144 = ComponentMask(None=TransformPosition(None=NamedRerouteUsage()) {transformSourceType=TRANSFORMPOSSOURCE_World, transformType=TRANSFORMPOSSOURCE_View}) {b=true}
  n145 = MaterialFunctionCall(UVs=NamedRerouteUsage(), Texture Scale=Multiply(A=Divide(B=@n144), B=ScalarParameter() {defaultValue=60, parameterName=sffff}) {constB=60}) {materialFunction=&ScaleUVsByCenter}
  n146 = ScalarParameter() {defaultValue=0.2, parameterName=InwardsRays_LengthScale}
  n147 = ScalarParameter() {defaultValue=2, parameterName=InwardRays_RadialScale}
  n148 = TextureSample(UVs=Panner(Coordinate=MaterialFunctionCall(UV Coord=@n145.UVs, LenghtScale=@n146, RadialScale=@n147) {materialFunction=&MF_PolarCoordinates}.Result) {speedX=0.5}) {
   samplerSource=SSM_Wrap_WorldGroupSettings
   texture=&T_AcademiaSL_T3_Masks_03
  }
  n149 = Multiply(A=@n148.R) {constB=4}
  NamedRerouteDeclaration(None=Add(A=Multiply(A=@n141, B=@n143), B=Multiply(A=@n149, B=@n142))) {name=InwardRays, nodeColor=<rgba 1 0.4453 0 1>}
  n150 = NamedRerouteUsage()
  n151 = TransformPosition(None=NamedRerouteUsage()) {transformSourceType=TRANSFORMPOSSOURCE_World, transformType=TRANSFORMPOSSOURCE_View}
  n152 = Multiply(
   A=Divide(B=ComponentMask(None=@n151) {b=true})
   B=ScalarParameter() {defaultValue=7, group=SUPERLEVEL, parameterName=ChestRays_Scale, sortPriority=101}
  )
  n153 = VectorParameter() {group=RelayStick_Orb, parameterName="OrbRadialCoordScales (UVUV)", defaultValue=<rgba 0.1366 2 0.174 1>}
  n154 = Multiply(
   A=MaterialFunctionCall(UV Coord=MaterialFunctionCall(UVs=@n150, Texture Scale=@n152) {materialFunction=&ScaleUVsByCenter}.UVs) {materialFunction=&MF_PolarCoordinates}.Result
   B=AppendVector(A=@n153.B, B=@n153.B)
  )
  n155 = Saturate(None=Multiply(A=ComponentMask(None=@n154) {r=true}) {constB=10})
  n156 = Time()
  n157 = VectorParameter() {group=RelayStick_Orb, parameterName="OrbRadialCoordPanSpeeds (UVUV)", defaultValue=<rgba 0.0333 0.0333 0.125 -0.0233>}
  n158 = AppendVector(A=@n157.B, B=@n157.B)
  n159 = TextureCoordinate() {uTiling=3, vTiling=3}
  n160 = TextureSample!(
   UVs=Add(A=Panner(Coordinate=@n154, Time=@n156, Speed=@n158) {speedX=-0.33, speedY=-0.053}) {constB=0}
   Tex
   "DDX(UVs)"=DDX(Value=@n159)
   "DDY(UVs)"=DDY(Value=@n159)
   Apply View MipBias
  ) {
   mipValueMode=TMVM_Derivative
   samplerSource=SSM_Wrap_WorldGroupSettings
   samplerType=SAMPLERTYPE_LinearColor
   texture=/Game/Characters/Player/Male/Medium/Bodies/M_MED_SRDrift/Textures/T_ElectricNoise
  }
  n161 = Multiply(A=Multiply(A=@n155, B=@n155), B=Power(Base=@n160.R, Exp=ScalarParameter() {defaultValue=2, group=SUPERLEVEL, parameterName=LightningPower}))
  NamedRerouteDeclaration(None=Multiply(A=Multiply(A=@n161) {constB=2}, B=NamedRerouteUsage())) {name=ChestRays, nodeColor=<rgba 0.1719 1 0 1>}
  n162 = NamedRerouteUsage()
  n163 = VectorParameter() {parameterName=CrotchMask_Position, defaultValue=<rgba 0 0 80 0>}
  n164 = ScalarParameter() {defaultValue=40, parameterName=CrotchMask_Radius}
  NamedRerouteDeclaration(None=OneMinus(None=SphereMask(A=@n162, B=@n163.RGB, Radius=@n164) {attenuationRadius=0, hardnessPercent=0.2})) {name=CrotchMask, nodeColor=<rgba 1 0 0.0469 1>}
  OneMinus(None=@n31)
 ]
}
