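# Parent character material for the LazyLizz Chip "Superlevel" hair style, built on the 2023 character parent functions for skeletal meshes.
# PreFX feeds the UV2 composite, then a static switch, on by default, applies the LazyLizz skin function.
# The resulting base color is desaturated and tinted by a powered-Fresnel blend between primary and Fresnel colors.
# That Fresnel color also adds scalable emissive rim glow before PostFX finishes the output.
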Material /BRCosmetics/Characters/Player/Female/Medium/Bodies/F_MED_SpatialTravel_Watt/Skins/Superlevels/ParentMaterials/M_LazyLizz_Chip_Hair_SL
graph {
 properties {
  bUseMaterialAttributes=true
  bUsedWithMeshDeformer=true
  bUsedWithMorphTargets=true
  bUsedWithSkeletalMesh=true
  shadingModel=MSM_FromMaterialExpression
  usedShadingModels=Default Lit | Subsurface | Clear Coat | Subsurface Profile | Cloth
 }
 nodes [
  n1 = MaterialFunctionCall() {materialFunction=&MF_CharacterParent_2023_PreFX}
  n2 = MaterialFunctionCall!(Material=@n1.Result, TechArtMask=@n1.Result) {materialFunction=/Game/Global/Materials/Character_Parent_2023/UV2Composite_Parent_2023/MF_UV2_Composite_2023}
  n3 = GetMaterialAttributes(None=StaticSwitchParameter(
    True=MaterialFunctionCall!(Material Attributes=@n2.Result, M=@n1.M, Fresnel=@n1.M, TA Mask=@n1.M) {materialFunction=/BRCosmetics/Characters/Player/Male/Medium/Bodies/M_MED_LazyLizz_Chip/Materials/MF_LazyLizz_Chip_Skin}.Result
    False=@n2.Result
   ) {defaultValue=true, group=04_Tech Art, parameterName=Use LazyLizz Skin}) {attributeGetTypes [#1, #2]}
  n4 = Desaturation(None=@n3.Base Color, Fraction=Constant() {r=1})
  n5 = VectorParameter() {group=superlevel, parameterName=Hair_FresnelColor, defaultValue=<rgba 0.551 0.1874 1 1>}
  n6 = Power(Base=@n1.Fresnel, Exp=ScalarParameter() {defaultValue=2, group=superlevel, parameterName=Hair_FresnelExponent})
  n7 = LinearInterpolate(A=VectorParameter() {group=superlevel, parameterName=Hair_PrimaryColor, defaultValue=<rgba 1 0.1549 0.881 1>}.RGB, B=@n5.RGB, Alpha=@n6)
  n8 = Multiply(A=Multiply(A=@n5.RGB, B=@n6), B=ScalarParameter() {group=superlevel, parameterName=Hair_EmissiveIntensity})
  MaterialFunctionCall(
   Material=SetMaterialAttributes!(MaterialAttributes=@n3.MaterialAttributes, Base Color=Multiply(A=@n4, B=@n7), Emissive Color=Add(A=@n3.Emissive Color, B=@n8)) {attributeSetTypes [#1, #2]}
   M=@n1.M
   Pre Skinned Local Position=@n1.M
   Distance Blend=@n1.M
   Fresnel=@n1.M
  ) {materialFunction=&MF_CharacterParent_2023_PostFX}
 ]
}
